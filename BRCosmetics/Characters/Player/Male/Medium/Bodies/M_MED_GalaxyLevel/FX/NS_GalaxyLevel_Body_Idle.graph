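# Niagara idle effect for the GalaxyLevel male body cosmetic: a persistent magenta glow and light pinned to the character's skeletal mesh.
# A single emitter burst-spawns non-expiring particles that follow a sampled bone or socket, drawn through sprite, component and light renderers.
# The User.Simualted? bool switches socket index, offset and facing axis, with facing rotated by the sampled rotation.
# A time-of-day scalar dims the particle color at night.
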NiagaraSystem /BRCosmetics/Characters/Player/Male/Medium/Bodies/M_MED_GalaxyLevel/FX/NS_GalaxyLevel_Body_Idle
dependencies {usedDataInterfaces [], usedDynamicInputs [], usedModules [&SystemState]}
emitters {Glow_and_Light {
  emitterUpdateScript [
   &EmitterState(Life Cycle Mode=::NewEnumerator1, Loop Behavior=::NewEnumerator1, Loop Duration Mode=::NewEnumerator1, UseLoopDelay=false)
   &SpawnBurst_Instantaneous(Use Spawn Probability=false, UseLoopCountLimit=false)
  ]
  particleSpawnScript [
   &InitializeParticle(
    Color=<rgba 1 0 0.9167 1>
    Lifetime=0.2
    Position Mode=::NewEnumerator0
    Sprite Size Mode=::NewEnumerator3
    Uniform Sprite Size=15
    Write Material Random=true
   )
   &CameraOffset(Camera Offset Amount=5)
  ]
  particleUpdateScript [
   &ParticleState(Kill Particles When Lifetime Has Elapsed=false, Let Infinitely Lived Particles Die When Emitter Deactivates=true)
   &SkeletalMeshLocation(
    Bone / Socket Index=&MakeCustomIntFromBool(Bool=Linked() {linkedVariable {name=User.Simualted?, type=NiagaraBool, defaultValue=~, description=~}}, True INT=1, False INT=0)
    Bone Sampling Mode=::NewEnumerator4
    MaskSpawnGroup=false
    Sampled Position Offset=&MakeCustomVectorFromBool(
     Bool=Linked() {linkedVariable {name=User.Simualted?, type=NiagaraBool, defaultValue=~, description=~}}
     True Vector=<-1.5 14.5 0>
     False Vector=<0.05 0 0>
    )
    Skeletal Mesh=$User.SkeletalMesh
   )
   &SpriteFacingAndAlignment(
    Sprite Alignment=<1 0 0>
    Sprite Facing=&RotateVectorByQuaternion(
     Vector To Rotate=&MakeCustomVectorFromBool(
      Bool=Linked() {linkedVariable {name=User.Simualted?, type=NiagaraBool, defaultValue=~, description=~}}
      True Vector=<0 1 0>
      False Vector=<1 0 0>
     )
     Quaternion=$Particles.SkeletalMeshLocation.SampledRotation
    )
    WriteAlignment=true
    WriteFacing=true
   )
   &ScaleColor(
    Scale RGB=&VectorFromFloat(Value=&FNTimeOfDayScalar(Night Scale=0.25, Morning Scale=1, Evening Scale=1))
    ScaleA=true
    ScaleRGB=true
   )
   /SetVariables(
    Particles.LightColor=&MultiplyLinearColor(Linear Color=$Particles.Color, Scale Factor=<10 10 10 1>)
    Particles.LightExponent=2
    Particles.LightRadius=8
    Particles.PointLightExponent=7
   ) {isSetParameters=true}
  ]
  rendererClasses [
   &NiagaraSpriteRendererProperties
   &NiagaraComponentRendererProperties
   &NiagaraLightRendererProperties
  ]
  renderers [
   {rendererClass=&NiagaraSpriteRendererProperties}
   {rendererClass=&NiagaraComponentRendererProperties, rendererIndex=1}
   {rendererClass=&NiagaraLightRendererProperties, rendererIndex=2}
  ]
 }}
summary {
 systemName=NS_GalaxyLevel_Body_Idle
 emitters [{
   emitterName=Glow_and_Light
   rendererClasses [
    &NiagaraSpriteRendererProperties
    &NiagaraComponentRendererProperties
    &NiagaraLightRendererProperties
   ]
  }]
 userVariables [
  User.Simualted? {type=NiagaraBool}
  User.SkeletalMesh {
   type=NiagaraDataInterfaceSkeletalMesh
   defaultValue {
    struct=NiagaraExt_VariableValue_DataInterface
    value {dataInterface=&self:NiagaraDataInterfaceSkeletalMesh_0}
   }
  }
 ]
}
