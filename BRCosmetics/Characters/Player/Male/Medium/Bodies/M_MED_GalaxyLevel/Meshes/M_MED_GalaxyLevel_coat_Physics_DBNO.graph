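# Physics asset for the GalaxyLevel male medium coat, the DBNO (downed) variant, simulating coat motion around the lower body.
# Seven capsule chains (three per side, one back) hang from pelvis-linked master bones, with tight per-segment swing and twist limits.
# Free-angular constraints cross-link neighboring panels at middle segments.
# Pelvis, leg and foot capsules form the body colliders, and a huge root sphere named floorCOL acts as the ground.
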PhysicsAsset /BRCosmetics/Characters/Player/Male/Medium/Bodies/M_MED_GalaxyLevel/Meshes/M_MED_GalaxyLevel_coat_Physics_DBNO
bodies {
 calf_l [{shapeType=Capsule, length=30, radius=9, center=<-16.0977 -2 0>, rotation=<rot 90 0 0>}]
 calf_r [{shapeType=Capsule, length=30, radius=9, center=<16.0977 2 0>, rotation=<rot 90 0 0>}]
 dyn_coat_01_L_01 [{shapeType=Capsule, length=15, radius=6, shapeName=dyn_coat_01_L_01_capsule, center=<5 0 5>, rotation=<rot 0.0001 0.0001 83.1482>}]
 dyn_coat_01_L_02 [{shapeType=Capsule, length=5, radius=10, shapeName=dyn_coat_01_L_02_capsule, center=<9 0 6>, rotation=<rot 0.0009 0.0013 88.1482>}]
 dyn_coat_01_L_03 [{shapeType=Capsule, length=5, radius=10, shapeName=dyn_coat_01_L_03_capsule, center=<6 0 8>, rotation=<rot 0.0008 0.0012 93.1482>}]
 dyn_coat_01_L_04 [{shapeType=Capsule, length=5, radius=10, shapeName=dyn_coat_01_L_04_capsule, center=<6 0 6>, rotation=<rot 0 0 93.6025>}]
 dyn_coat_01_L_mstr [{shapeType=Capsule, length=2, radius=2, shapeName=dyn_coat_01_L_mstr_capsule}]
 dyn_coat_01_R_01 [{shapeType=Capsule, length=20, radius=6, shapeName=dyn_coat_01_R_01_capsule, center=<6 0 6>, rotation=<rot 0.0002 -0.0002 90>}]
 dyn_coat_01_R_02 [{shapeType=Capsule, length=5, radius=10, shapeName=dyn_coat_01_R_02_capsule, center=<3 0 10>, rotation=<rot -0.0002 179.9998 90>}]
 dyn_coat_01_R_03 [{shapeType=Capsule, length=5, radius=10, shapeName=dyn_coat_01_R_03_capsule, center=<6 0 10>, rotation=<rot -0.0002 179.9998 90>}]
 dyn_coat_01_R_04 [{shapeType=Capsule, length=5, radius=9, shapeName=dyn_coat_01_R_04_capsule, center=<6 0 5>, rotation=<rot 0 0 90>}]
 dyn_coat_01_R_mstr [{shapeType=Capsule, length=2, radius=2, shapeName=dyn_coat_01_R_mstr_capsule}]
 dyn_coat_02_L_01 [{shapeType=Capsule, length=15, radius=5, shapeName=dyn_coat_02_L_01_capsule, center=<6 0 6>, rotation=<rot 0 0 90>}]
 dyn_coat_02_L_02 [{shapeType=Capsule, length=10, radius=9, shapeName=dyn_coat_02_L_02_capsule, center=<6 0 6>, rotation=<rot 0 0 90>}]
 dyn_coat_02_L_03 [{shapeType=Capsule, length=10, radius=10.5, shapeName=dyn_coat_02_L_03_capsule, center=<6 0 6>, rotation=<rot 0 0 90>}]
 dyn_coat_02_L_04 [{shapeType=Capsule, length=10, radius=8, shapeName=dyn_coat_02_L_04_capsule, center=<6 0 6>, rotation=<rot 0 0 90>}]
 dyn_coat_02_L_mstr [{shapeType=Capsule, length=2, radius=2, shapeName=dyn_coat_02_L_mstr_capsule}]
 dyn_coat_02_R_01 [{shapeType=Capsule, length=20, radius=5, shapeName=dyn_coat_02_R_01_capsule, center=<6 0 5>, rotation=<rot 0 0 90>}]
 dyn_coat_02_R_02 [{shapeType=Capsule, length=20, radius=7, shapeName=dyn_coat_02_R_02_capsule, center=<6 0 5>, rotation=<rot 0 0 90>}]
 dyn_coat_02_R_03 [{shapeType=Capsule, length=20, radius=8, shapeName=dyn_coat_02_R_03_capsule, center=<6 0 5>, rotation=<rot 0 0 90>}]
 dyn_coat_02_R_04 [{shapeType=Capsule, length=28, radius=7, shapeName=dyn_coat_02_R_04_capsule, center=<6 0 5>, rotation=<rot 0 0 90>}]
 dyn_coat_02_R_mstr [{shapeType=Capsule, length=2, radius=2, shapeName=dyn_coat_02_R_mstr_capsule}]
 dyn_coat_03_L_01 [{shapeType=Capsule, length=20, radius=8, shapeName=dyn_coat_03_L_01_capsule, center=<4 0 5>, rotation=<rot 0 0 90>}]
 dyn_coat_03_L_02 [{shapeType=Capsule, length=24, radius=9, shapeName=dyn_coat_03_L_02_capsule, center=<6 0 5>, rotation=<rot 0 0 90>}]
 dyn_coat_03_L_03 [{shapeType=Capsule, length=25, radius=10, shapeName=dyn_coat_03_L_03_capsule, center=<6 0 4>, rotation=<rot 0 0 90>}]
 dyn_coat_03_L_04 [{shapeType=Capsule, length=25, radius=11.5, shapeName=dyn_coat_03_L_04_capsule, center=<6 0 3>, rotation=<rot 0 0 90>}]
 dyn_coat_03_L_mstr [{shapeType=Capsule, length=2, radius=2, shapeName=dyn_coat_03_L_mstr_capsule}]
 dyn_coat_03_R_01 [{shapeType=Capsule, length=20, radius=8, shapeName=dyn_coat_03_R_01_capsule, center=<4 0 5>, rotation=<rot 0 0 90>}]
 dyn_coat_03_R_02 [{shapeType=Capsule, length=24, radius=9, shapeName=dyn_coat_03_R_02_capsule, center=<6 0 6>, rotation=<rot 0 0 90>}]
 dyn_coat_03_R_03 [{shapeType=Capsule, length=25, radius=10, shapeName=dyn_coat_03_R_03_capsule, center=<6 0 5>, rotation=<rot 0 0 90>}]
 dyn_coat_03_R_04 [{shapeType=Capsule, length=25, radius=11.5, shapeName=dyn_coat_03_R_04_capsule, center=<6 0 5>, rotation=<rot 0 0 90>}]
 dyn_coat_03_R_mstr [{shapeType=Capsule, length=2, radius=2, shapeName=dyn_coat_03_R_mstr_capsule}]
 dyn_coat_B_01 [{shapeType=Capsule, length=20, radius=8, shapeName=dyn_coat_B_01_capsule, center=<4 0 5>, rotation=<rot 0 0 90>}]
 dyn_coat_B_02 [{shapeType=Capsule, length=24, radius=8, shapeName=dyn_coat_B_02_capsule, center=<6 0 5>, rotation=<rot 0 0 90>}]
 dyn_coat_B_03 [{shapeType=Capsule, length=25, radius=10, shapeName=dyn_coat_B_03_capsule, center=<6 0 3>, rotation=<rot 0 0 90>}]
 dyn_coat_B_04 [{shapeType=Capsule, length=25, radius=11.5, shapeName=dyn_coat_B_04_capsule, center=<6 0 3>, rotation=<rot 0 0 90>}]
 dyn_coat_B_mstr [{shapeType=Capsule, length=2, radius=2, shapeName=dyn_coat_B_mstr_capsule}]
 foot_l [{shapeType=Capsule, length=10.4337, radius=12, shapeName=foot_l_capsule, center=<-1.7794 3.3956 -0.4957>, rotation=<rot 0 177.8713 -88.3255>}]
 foot_r [{shapeType=Capsule, length=11, radius=12, shapeName=foot_r_capsule, center=<5.6997 -5.5023 0.4562>, rotation=<rot 0 171.9148 -88.9725>}]
 pelvis [{shapeType=Capsule, length=23.4001, radius=12, rotation=<rot 90 180 177.6663>}]
 root [{shapeType=Capsule, length=0, radius=210, shapeName=floorCOL_sphere, center=<0 15 -193>, rotation=<rot 0 -90 0>}]
 thigh_l [{shapeType=Capsule, length=26, radius=9, center=<-20.1442 0.2001 0.3862>, rotation=<rot 90 0 0>}]
 thigh_r [{shapeType=Capsule, length=28, radius=9, shapeName=thigh_r_capsule, center=<21.0965 -0.5463 -0.0801>, rotation=<rot 90 0 0>}]
}
constraints [
 {bone1Name=thigh_l, bone2Name=pelvis, swing1LimitDegrees=45, swing2LimitDegrees=45, twistLimitDegrees=45}
 {bone1Name=calf_l, bone2Name=thigh_l, swing1LimitDegrees=45, swing2LimitDegrees=45, twistLimitDegrees=45}
 {bone1Name=calf_r, bone2Name=thigh_r, swing1LimitDegrees=45, swing2LimitDegrees=45, twistLimitDegrees=45}
 {
  bone1Name=pelvis
  bone2Name=root
  swing1LimitDegrees=45
  swing1Motion=Free
  swing2LimitDegrees=45
  swing2Motion=Free
  twistLimitDegrees=45
  twistMotion=Free
 }
 {
  bone1Name=thigh_r
  bone2Name=pelvis
  swing1LimitDegrees=45
  swing1Motion=Free
  swing2LimitDegrees=45
  swing2Motion=Free
  twistLimitDegrees=45
  twistMotion=Free
 }
 {
  bone1Name=dyn_coat_01_L_mstr
  bone2Name=pelvis
  swing1LimitDegrees=45
  swing1Motion=Free
  swing2LimitDegrees=45
  swing2Motion=Free
  twistLimitDegrees=45
  twistMotion=Free
 }
 {bone1Name=dyn_coat_01_L_01, bone2Name=dyn_coat_01_L_mstr, swing1LimitDegrees=15, swing2LimitDegrees=55, twistLimitDegrees=15}
 {bone1Name=dyn_coat_01_L_02, bone2Name=dyn_coat_01_L_01, swing1LimitDegrees=10, swing2LimitDegrees=60, twistLimitDegrees=15}
 {bone1Name=dyn_coat_01_L_03, bone2Name=dyn_coat_01_L_02, swing1LimitDegrees=5, swing2LimitDegrees=45, twistLimitDegrees=5}
 {bone1Name=dyn_coat_01_L_04, bone2Name=dyn_coat_01_L_03, swing1LimitDegrees=15, swing2LimitDegrees=45, twistLimitDegrees=2}
 {
  bone1Name=dyn_coat_01_R_mstr
  bone2Name=pelvis
  swing1LimitDegrees=45
  swing1Motion=Free
  swing2LimitDegrees=45
  swing2Motion=Free
  twistLimitDegrees=45
  twistMotion=Free
 }
 {bone1Name=dyn_coat_01_R_01, bone2Name=dyn_coat_01_R_mstr, swing1LimitDegrees=15, swing2LimitDegrees=55, twistLimitDegrees=10}
 {bone1Name=dyn_coat_01_R_02, bone2Name=dyn_coat_01_R_01, swing1LimitDegrees=15, swing2LimitDegrees=35, twistLimitDegrees=15}
 {bone1Name=dyn_coat_01_R_03, bone2Name=dyn_coat_01_R_02, swing1LimitDegrees=20, swing2LimitDegrees=30, twistLimitDegrees=5}
 {bone1Name=dyn_coat_01_R_04, bone2Name=dyn_coat_01_R_03, swing1LimitDegrees=25, swing2LimitDegrees=30, twistLimitDegrees=2}
 {
  bone1Name=dyn_coat_02_L_mstr
  bone2Name=pelvis
  swing1LimitDegrees=45
  swing1Motion=Free
  swing2LimitDegrees=45
  swing2Motion=Free
  twistLimitDegrees=45
  twistMotion=Free
 }
 {bone1Name=dyn_coat_02_L_01, bone2Name=dyn_coat_02_L_mstr, swing1LimitDegrees=25, swing2LimitDegrees=55, twistLimitDegrees=10}
 {bone1Name=dyn_coat_02_L_02, bone2Name=dyn_coat_02_L_01, swing1LimitDegrees=20, swing2LimitDegrees=30, twistLimitDegrees=10}
 {bone1Name=dyn_coat_02_L_03, bone2Name=dyn_coat_02_L_02, swing1LimitDegrees=20, swing2LimitDegrees=30, twistLimitDegrees=5}
 {bone1Name=dyn_coat_02_L_04, bone2Name=dyn_coat_02_L_03, swing1LimitDegrees=25, swing2LimitDegrees=30, twistLimitDegrees=2}
 {
  bone1Name=dyn_coat_02_R_mstr
  bone2Name=pelvis
  swing1LimitDegrees=45
  swing1Motion=Free
  swing2LimitDegrees=45
  swing2Motion=Free
  twistLimitDegrees=45
  twistMotion=Free
 }
 {bone1Name=dyn_coat_02_R_01, bone2Name=dyn_coat_02_R_mstr, swing1LimitDegrees=15, swing2LimitDegrees=30, twistLimitDegrees=3}
 {bone1Name=dyn_coat_02_R_02, bone2Name=dyn_coat_02_R_01, swing1LimitDegrees=15, swing2LimitDegrees=30, twistLimitDegrees=3}
 {bone1Name=dyn_coat_02_R_03, bone2Name=dyn_coat_02_R_02, swing1LimitDegrees=10, swing2LimitDegrees=30, twistLimitDegrees=2}
 {bone1Name=dyn_coat_02_R_04, bone2Name=dyn_coat_02_R_03, swing1LimitDegrees=15, swing2LimitDegrees=30, twistLimitDegrees=2}
 {
  bone1Name=dyn_coat_03_L_mstr
  bone2Name=pelvis
  swing1LimitDegrees=45
  swing1Motion=Free
  swing2LimitDegrees=45
  swing2Motion=Free
  twistLimitDegrees=45
  twistMotion=Free
 }
 {bone1Name=dyn_coat_03_L_01, bone2Name=dyn_coat_03_L_mstr, swing1LimitDegrees=15, swing2LimitDegrees=20, twistLimitDegrees=2}
 {bone1Name=dyn_coat_03_L_02, bone2Name=dyn_coat_03_L_01, swing1LimitDegrees=15, swing2LimitDegrees=30, twistLimitDegrees=2}
 {bone1Name=dyn_coat_03_L_03, bone2Name=dyn_coat_03_L_02, swing1LimitDegrees=20, swing2LimitDegrees=30, twistLimitDegrees=2}
 {bone1Name=dyn_coat_03_L_04, bone2Name=dyn_coat_03_L_03, swing1LimitDegrees=25, swing2LimitDegrees=30, twistLimitDegrees=2}
 {
  bone1Name=dyn_coat_03_R_mstr
  bone2Name=pelvis
  swing1LimitDegrees=45
  swing1Motion=Free
  swing2LimitDegrees=45
  swing2Motion=Free
  twistLimitDegrees=45
  twistMotion=Free
 }
 {bone1Name=dyn_coat_03_R_01, bone2Name=dyn_coat_03_R_mstr, swing1LimitDegrees=30, swing2LimitDegrees=20, twistLimitDegrees=5}
 {bone1Name=dyn_coat_03_R_02, bone2Name=dyn_coat_03_R_01, swing1LimitDegrees=20, swing2LimitDegrees=30, twistLimitDegrees=5}
 {bone1Name=dyn_coat_03_R_03, bone2Name=dyn_coat_03_R_02, swing1LimitDegrees=20, swing2LimitDegrees=30, twistLimitDegrees=5}
 {bone1Name=dyn_coat_03_R_04, bone2Name=dyn_coat_03_R_03, swing1LimitDegrees=15, swing2LimitDegrees=30, twistLimitDegrees=5}
 {
  bone1Name=dyn_coat_B_mstr
  bone2Name=pelvis
  swing1LimitDegrees=45
  swing1Motion=Free
  swing2LimitDegrees=45
  swing2Motion=Free
  twistLimitDegrees=45
  twistMotion=Free
 }
 {bone1Name=dyn_coat_B_01, bone2Name=dyn_coat_B_mstr, swing1LimitDegrees=10, swing2LimitDegrees=20, twistLimitDegrees=5}
 {bone1Name=dyn_coat_B_02, bone2Name=dyn_coat_B_01, swing1LimitDegrees=10, swing2LimitDegrees=30, twistLimitDegrees=5}
 {bone1Name=dyn_coat_B_03, bone2Name=dyn_coat_B_02, swing1LimitDegrees=10, swing2LimitDegrees=30, twistLimitDegrees=10}
 {bone1Name=dyn_coat_B_04, bone2Name=dyn_coat_B_03, swing1LimitDegrees=15, swing2LimitDegrees=30, twistLimitDegrees=10}
 {
  bone1Name=foot_r
  bone2Name=calf_r
  swing1LimitDegrees=45
  swing1Motion=Free
  swing2LimitDegrees=45
  swing2Motion=Free
  twistLimitDegrees=45
  twistMotion=Free
 }
 {
  bone1Name=foot_l
  bone2Name=calf_l
  swing1LimitDegrees=45
  swing1Motion=Free
  swing2LimitDegrees=45
  swing2Motion=Free
  twistLimitDegrees=45
  twistMotion=Free
 }
 {
  bone1Name=dyn_coat_02_R_02
  bone2Name=dyn_coat_01_R_02
  swing1LimitDegrees=45
  swing1Motion=Free
  swing2LimitDegrees=45
  swing2Motion=Free
  twistLimitDegrees=45
  twistMotion=Free
 }
 {
  bone1Name=dyn_coat_03_L_03
  bone2Name=dyn_coat_B_03
  swing1LimitDegrees=45
  swing1Motion=Free
  swing2LimitDegrees=45
  swing2Motion=Free
  twistLimitDegrees=45
  twistMotion=Free
 }
 {
  bone1Name=dyn_coat_B_03
  bone2Name=dyn_coat_03_L_03
  swing1LimitDegrees=45
  swing1Motion=Free
  swing2LimitDegrees=45
  swing2Motion=Free
  twistLimitDegrees=45
  twistMotion=Free
 }
 {
  bone1Name=dyn_coat_02_R_03
  bone2Name=dyn_coat_03_R_03
  swing1LimitDegrees=45
  swing1Motion=Free
  swing2LimitDegrees=45
  swing2Motion=Free
  twistLimitDegrees=45
  twistMotion=Free
 }
 {
  bone1Name=dyn_coat_B_03
  bone2Name=dyn_coat_03_R_03
  swing1LimitDegrees=45
  swing1Motion=Free
  swing2LimitDegrees=45
  swing2Motion=Free
  twistLimitDegrees=45
  twistMotion=Free
 }
 {
  bone1Name=dyn_coat_02_L_03
  bone2Name=dyn_coat_01_L_03
  swing1LimitDegrees=45
  swing1Motion=Free
  swing2LimitDegrees=45
  swing2Motion=Free
  twistLimitDegrees=45
  twistMotion=Free
 }
 {
  bone1Name=dyn_coat_03_L_03
  bone2Name=dyn_coat_02_L_03
  swing1LimitDegrees=45
  swing1Motion=Free
  swing2LimitDegrees=45
  swing2Motion=Free
  twistLimitDegrees=45
  twistMotion=Free
 }
 {
  bone1Name=dyn_coat_02_R_02
  bone2Name=dyn_coat_03_R_02
  swing1LimitDegrees=45
  swing1Motion=Free
  swing2LimitDegrees=45
  swing2Motion=Free
  twistLimitDegrees=45
  twistMotion=Free
 }
 {
  bone1Name=dyn_coat_02_R_03
  bone2Name=dyn_coat_01_R_03
  swing1LimitDegrees=45
  swing1Motion=Free
  swing2LimitDegrees=45
  swing2Motion=Free
  twistLimitDegrees=45
  twistMotion=Free
 }
]
