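# Physics asset for the necklace mesh of the M_MED_GalaxyLevel male player body, covering its dangling chain and medal.
# Five capsule bodies form a series from spine_03 through dyn_chain_mstr, dyn_chain and dyn_medal_mst to dyn_medal.
# The chain master rotates freely on the spine, while chain and medal each swing within narrow limits on their small master bodies.
# The medal master is rotation-locked to the chain, so it follows the chain's motion.
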PhysicsAsset /BRCosmetics/Characters/Player/Male/Medium/Bodies/M_MED_GalaxyLevel/Meshes/M_MED_GalaxyLevel_necklace_Physics
bodies {
 dyn_chain [{shapeType=Capsule, length=6.2504, radius=4.6681, shapeName=dyn_chain_capsule, center=<3 0 0>}]
 dyn_chain_mstr [{shapeType=Capsule, length=0.505, radius=0.505, shapeName=dyn_chain_mstr_capsule, rotation=<rot 0 -90 0>}]
 dyn_medal [{shapeType=Capsule, length=5.7699, radius=4.6955, shapeName=dyn_medal_capsule, center=<5 0 0>}]
 dyn_medal_mst [{shapeType=Capsule, length=0.505, radius=0.505, shapeName=dyn_medal_mst_capsule, rotation=<rot 0 -90 0>}]
 spine_03 [{shapeType=Capsule, length=17.4452, radius=10, shapeName=spine_03_capsule, center=<6.1915 1.6052 0.433>, rotation=<rot -4.1614 -90 0.6729>}]
}
constraints [
 {
  bone1Name=dyn_medal_mst
  bone2Name=dyn_chain
  swing1LimitDegrees=45
  swing1Motion=Locked
  swing2LimitDegrees=45
  swing2Motion=Locked
  twistLimitDegrees=45
  twistMotion=Locked
 }
 {bone1Name=dyn_medal, bone2Name=dyn_medal_mst, swing1LimitDegrees=35, swing2LimitDegrees=10, twistLimitDegrees=10}
 {bone1Name=dyn_chain, bone2Name=dyn_chain_mstr, swing1LimitDegrees=15, swing2LimitDegrees=5, twistLimitDegrees=3}
 {
  bone1Name=dyn_chain_mstr
  bone2Name=spine_03
  swing1LimitDegrees=45
  swing1Motion=Free
  swing2LimitDegrees=45
  swing2Motion=Free
  twistLimitDegrees=45
  twistMotion=Free
 }
]
