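# Control Rig for the GalaxyLevel male body that corrects lapel, necklace chain and coat dyn bones, apparently as a post-process.
# A three-branch sequence offsets lapels by medal distance and chain by medal rotation, then coat master bones by spherical pose readers, then by bone distances and rotation.
# Clamped remaps and reader outputs weight each transform modification.
# Several nodes are unwired leftovers, including a debug-drawn pose reader and a head-driven lapel tweak.
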ControlRigBlueprint /BRCosmetics/Characters/Player/Male/Medium/Bodies/M_MED_GalaxyLevel/Meshes/M_MED_GalaxyLevel_necklacepost_CtrlRig
graphs [
 ContainedGraph {nodes [
   n1 = RigVMFunctionEntryNode()
   SequenceAggregate = &RigVMFunction_Sequence(ExecuteContext=@n1.ExecuteContext)
   SequenceAggregate_1 = &RigVMFunction_Sequence(ExecuteContext=@SequenceAggregate.B)
   RigVMFunctionReturnNode(A=@SequenceAggregate.A, B=@SequenceAggregate_1.A, C=@SequenceAggregate_1.B)
  ]}
 RigVMFunctionLibrary {links [], nodes []}
 RigVMModel {nodes [
   RigUnit_BeginExecution = &RigUnit_BeginExecution()
   SequenceAggregate = RigVMAggregateNode(ExecuteContext=@RigUnit_BeginExecution.ExecutePin)
   n1 = &RigUnit_GetTransform(Item=(Type=Bone,Name="dyn_lapel_l"))
   n2 = &RigUnit_GetTransform(Item=(Type=Control,Name="medal loc"))
   &RigUnit_ModifyTransforms(
    ItemToModify=((Item=(Type=Bone,Name="dyn_lapel_l"),Transform=(Rotation=(X=0,Y=0,Z=0,W=1),Translation=(X=9,Y=-6,Z=0),Scale3D=(X=1,Y=1,Z=1))))
    Weight=&RigVMFunction_MathFloatRemap(
     SourceMaximum=3
     SourceMinimum=7
     TargetMinimum=0
     bClamp=true
     Value=&RigVMFunction_MathVectorDistance(B=@n1.Transform.Translation, A=@n2.Transform.Translation).Result
    ).Result
    ItemToModify.0.Transform.Rotation=&RigVMFunction_MathQuaternionFromEuler(Euler=(X=45,Y=0,Z=-90)).Result
    ExecutePin=@SequenceAggregate.A
   )
   n3 = &RigUnit_GetTransform(Item=(Type=Bone,Name="dyn_rapel_r"))
   n4 = &RigUnit_GetTransform(Item=(Type=Control,Name="medal loc"))
   > &RigUnit_ModifyTransforms(
    ItemToModify=((Item=(Type=Bone,Name="dyn_rapel_r"),Transform=(Rotation=(X=0,Y=0,Z=0,W=1),Translation=(X=8,Y=8,Z=0),Scale3D=(X=1,Y=1,Z=1))))
    Weight=&RigVMFunction_MathFloatRemap(
     SourceMaximum=2
     SourceMinimum=8
     TargetMinimum=0
     bClamp=true
     Value=&RigVMFunction_MathVectorDistance(B=@n3.Transform.Translation, A=@n4.Transform.Translation).Result
    ).Result
    ItemToModify.0.Transform.Rotation=&RigVMFunction_MathQuaternionFromEuler(Euler=(X=0,Y=0,Z=120)).Result
   )
   n5 = &RigVMFunction_MathQuaternionToEuler(
    RotationOrder=ZYX
    Value=&RigUnit_GetRelativeTransformForItem(Child=(Type=Bone,Name="dyn_medal"), Parent=(Type=Bone,Name="spine_04"), bChildInitial=false, bParentInitial=false).RelativeTransform.Rotation
   )
   > &RigUnit_ModifyTransforms(
    ItemToModify=((Item=(Type=Bone,Name="dyn_chain_mstr"),Transform=(Rotation=(X=0,Y=0,Z=0,W=1),Translation=(X=0,Y=0,Z=0),Scale3D=(X=1,Y=1,Z=1))))
    Weight=&RigVMFunction_MathFloatRemap(
     SourceMaximum=-167
     SourceMinimum=-164
     TargetMinimum=0
     bClamp=true
     Value=&RigVMFunction_MathFloatAdd(B=0, A=@n5.Result.Z) {template="Add::Execute(in A,in B,out Result)"}.Result
    ).Result
    ItemToModify.0.Transform.Rotation=&RigVMFunction_MathQuaternionFromEuler(Euler=(X=8,Y=0,Z=0)).Result
   )
   n6 = &RigUnit_SphericalPoseReader(
    ActiveRegionScaleFactors=(PositiveWidth=1,NegativeWidth=5,PositiveHeight=3,NegativeHeight=3)
    ActiveRegionSize=0.1
    Debug=(bDrawDebug=false,DebugScale=25,DebugSegments=20,DebugThickness=0.25)
    DriverAxis=(X=1,Y=0,Z=0)
    DriverItem=(Type=Bone,Name="dyn_coat_02_R_01")
    FalloffRegionScaleFactors=(PositiveWidth=0.6,NegativeWidth=2,PositiveHeight=3,NegativeHeight=2)
    FalloffSize=0.5
    OptionalParentItem=(Type=Bone,Name="pelvis")
    RotationOffset=(X=0,Y=-20,Z=0)
    ExecutePin=@SequenceAggregate.B
   )
   > &RigUnit_ModifyTransforms(
    ItemToModify=((Item=(Type=Bone,Name="dyn_coat_02_R_mstr"),Transform=(Rotation=(X=0,Y=0,Z=0,W=1),Translation=(X=-2,Y=0,Z=0),Scale3D=(X=1,Y=1,Z=1))))
    Weight=@n6.OutputParam
   )
   > &RigUnit_ModifyTransforms(
    ItemToModify=((Item=(Type=Bone,Name="dyn_coat_03_R_mstr"),Transform=(Rotation=(X=0,Y=0,Z=0,W=1),Translation=(X=0,Y=0,Z=0),Scale3D=(X=1,Y=1,Z=1))))
    Weight=@n6.OutputParam
   )
   > n7 = &RigUnit_SphericalPoseReader(
    ActiveRegionScaleFactors=(PositiveWidth=1,NegativeWidth=4,PositiveHeight=3,NegativeHeight=3)
    ActiveRegionSize=0.1
    Debug=(bDrawDebug=false,DebugScale=25,DebugSegments=20,DebugThickness=0.25)
    DriverAxis=(X=1,Y=0,Z=0)
    DriverItem=(Type=Bone,Name="dyn_coat_02_L_01")
    FalloffRegionScaleFactors=(PositiveWidth=0.6,NegativeWidth=0.2,PositiveHeight=2,NegativeHeight=2)
    FalloffSize=0.5
    OptionalParentItem=(Type=Bone,Name="pelvis")
    RotationOffset=(X=0,Y=-25,Z=0)
   )
   > &RigUnit_ModifyTransforms(
    ItemToModify=((Item=(Type=Bone,Name="dyn_coat_02_L_mstr"),Transform=(Rotation=(X=0,Y=0,Z=0,W=1),Translation=(X=3,Y=0,Z=2),Scale3D=(X=1,Y=1,Z=1))))
    Weight=@n7.OutputParam
   )
   > n8 = &RigUnit_SphericalPoseReader(
    ActiveRegionScaleFactors=(PositiveWidth=1,NegativeWidth=1,PositiveHeight=1,NegativeHeight=1)
    ActiveRegionSize=0.1
    Debug=(bDrawDebug=false,DebugScale=25,DebugSegments=20,DebugThickness=0.25)
    DriverAxis=(X=1,Y=0,Z=0)
    DriverItem=(Type=Bone,Name="dyn_coat_B_01")
    FalloffRegionScaleFactors=(PositiveWidth=1,NegativeWidth=1,PositiveHeight=1,NegativeHeight=1)
    FalloffSize=0.2
    OptionalParentItem=(Type=Bone,Name="pelvis")
    RotationOffset=(X=0,Y=-127,Z=0)
   )
   > &RigUnit_ModifyTransforms(
    ItemToModify=((Item=(Type=Bone,Name="dyn_coat_B_mstr"),Transform=(Rotation=(X=0,Y=0,Z=0,W=1),Translation=(X=0,Y=-2,Z=2),Scale3D=(X=1,Y=1,Z=1))))
    Weight=@n8.OutputParam
   )
   > &RigUnit_ModifyTransforms(
    ItemToModify=((Item=(Type=Bone,Name="dyn_coat_03_R_mstr"),Transform=(Rotation=(X=0,Y=0,Z=0,W=1),Translation=(X=-2,Y=0,Z=-2),Scale3D=(X=1,Y=1,Z=1))))
    Weight=@n8.OutputParam
   )
   > &RigUnit_ModifyTransforms(
    ItemToModify=((Item=(Type=Bone,Name="dyn_coat_03_L_mstr"),Transform=(Rotation=(X=0,Y=0,Z=0,W=1),Translation=(X=2,Y=0,Z=2),Scale3D=(X=1,Y=1,Z=1))))
    Weight=@n8.OutputParam
   )
   n9 = &RigUnit_GetTransform(Item=(Type=Bone,Name="dyn_coat_01_R_01"))
   n10 = &RigUnit_GetTransform(Item=(Type=Bone,Name="spine_02"))
   &RigUnit_ModifyTransforms(
    ItemToModify=((Item=(Type=Bone,Name="dyn_coat_01_R_mstr"),Transform=(Rotation=(X=0,Y=0,Z=0,W=1),Translation=(X=4,Y=-4,Z=0),Scale3D=(X=1,Y=1,Z=1))))
    Weight=&RigVMFunction_MathFloatRemap(
     SourceMaximum=10
     SourceMinimum=12
     TargetMinimum=0
     bClamp=true
     Value=&RigVMFunction_MathVectorDistance(B=@n9.Transform.Translation, A=@n10.Transform.Translation).Result
    ).Result
    ExecutePin=@SequenceAggregate.C
   )
   n11 = &RigUnit_GetTransform(Item=(Type=Bone,Name="dyn_coat_02_R_01"))
   n12 = &RigUnit_GetTransform(Item=(Type=Bone,Name="pelvis"))
   > &RigUnit_ModifyTransforms(
    ItemToModify=((Item=(Type=Bone,Name="dyn_coat_02_R_mstr"),Transform=(Rotation=(X=0,Y=0,Z=0,W=1),Translation=(X=-9,Y=0,Z=0),Scale3D=(X=1,Y=1,Z=1))))
    Weight=&RigVMFunction_MathFloatRemap(
     SourceMaximum=10
     SourceMinimum=15
     TargetMinimum=0
     bClamp=true
     Value=&RigVMFunction_MathVectorDistance(B=@n11.Transform.Translation, A=@n12.Transform.Translation).Result
    ).Result
   )
   n13 = &RigUnit_GetRelativeTransformForItem(Child=(Type=Bone,Name="dyn_coat_02_R_01"), Parent=(Type=Bone,Name="dyn_coat_02_R_mstr"), bChildInitial=false, bParentInitial=false)
   n14 = &RigVMFunction_MathQuaternionToEuler(RotationOrder=ZYX, Value=@n13.RelativeTransform.Rotation)
   > &RigUnit_ModifyTransforms(
    ItemToModify=((Item=(Type=Bone,Name="dyn_coat_02_R_mstr"),Transform=(Rotation=(X=0,Y=0,Z=0,W=1),Translation=(X=-2,Y=2,Z=0),Scale3D=(X=1,Y=1,Z=1))))
    Weight=&RigVMFunction_MathFloatRemap(
     SourceMaximum=0
     SourceMinimum=-60
     TargetMinimum=0
     bClamp=true
     Value=&RigVMFunction_MathFloatAdd(B=0, A=@n14.Result.Y) {template="Add::Execute(in A,in B,out Result)"}.Result
    ).Result
   )
   n15 = &RigUnit_GetTransform(Item=(Type=Bone,Name="dyn_coat_03_R_01"))
   n16 = &RigUnit_GetTransform(Item=(Type=Bone,Name="dyn_coat_03_R_mstr"))
   > &RigUnit_ModifyTransforms(
    ItemToModify=((Item=(Type=Bone,Name="dyn_coat_03_R_mstr"),Transform=(Rotation=(X=0,Y=0,Z=0,W=1),Translation=(X=-4,Y=0,Z=0),Scale3D=(X=1,Y=1,Z=1))))
    Weight=&RigVMFunction_MathFloatRemap(
     SourceMaximum=3
     SourceMinimum=0
     TargetMinimum=0
     bClamp=true
     Value=&RigVMFunction_MathVectorDistance(B=@n15.Transform.Translation, A=@n16.Transform.Translation).Result
    ).Result
   )
   n17 = &RigUnit_GetTransform(Item=(Type=Bone,Name="dyn_coat_B_01"))
   n18 = &RigUnit_GetTransform(Item=(Type=Bone,Name="pelvis"))
   > &RigUnit_ModifyTransforms(
    ItemToModify="((Item=(Type=Bone,Name=\"dyn_coat_B_mstr\"),Transform=(Rotation=(X=0,Y=0,Z=0,W=1),Translation=(X=0,Y=-12,Z=0),Scale3D=(X=1,Y=1,Z=1))),(Item=(Type=Bone,Name=\"dyn_coat_03_R_mstr\"),Transform=(Rotation=(X=0,Y=0,Z=0,W=1),Translation=(X=-12,Y=3,Z=0),Scale3D=(X=1,Y ...(truncated)"
    Weight=&RigVMFunction_MathFloatRemap(
     SourceMaximum=5
     SourceMinimum=13.5
     TargetMinimum=0
     bClamp=true
     Value=&RigVMFunction_MathVectorDistance(B=@n17.Transform.Translation, A=@n18.Transform.Translation).Result
    ).Result
   )
   n19 = &RigUnit_GetTransform(Item=(Type=Bone,Name="dyn_coat_01_L_01"))
   n20 = &RigUnit_GetTransform(Item=(Type=Bone,Name="spine_02"))
   > &RigUnit_ModifyTransforms(
    ItemToModify=((Item=(Type=Bone,Name="dyn_coat_01_L_mstr"),Transform=(Rotation=(X=0,Y=0,Z=0,W=1),Translation=(X=3,Y=3,Z=0),Scale3D=(X=1,Y=1,Z=1))))
    Weight=&RigVMFunction_MathFloatRemap(
     SourceMaximum=10
     SourceMinimum=12
     TargetMinimum=0
     bClamp=true
     Value=&RigVMFunction_MathVectorDistance(B=@n19.Transform.Translation, A=@n20.Transform.Translation).Result
    ).Result
   )
   n21 = &RigUnit_GetTransform(Item=(Type=Bone,Name="dyn_coat_02_L_01"))
   n22 = &RigUnit_GetTransform(Item=(Type=Bone,Name="spine_01"))
   > &RigUnit_ModifyTransforms(
    ItemToModify=((Item=(Type=Bone,Name="dyn_coat_02_L_mstr"),Transform=(Rotation=(X=0,Y=0,Z=0,W=1),Translation=(X=3,Y=0,Z=0),Scale3D=(X=1,Y=1,Z=1))))
    Weight=&RigVMFunction_MathFloatRemap(
     SourceMaximum=12
     SourceMinimum=15
     TargetMinimum=0
     bClamp=true
     Value=&RigVMFunction_MathVectorDistance(B=@n21.Transform.Translation, A=@n22.Transform.Translation).Result
    ).Result
   )
   n23 = &RigUnit_GetTransform(Item=(Type=Bone,Name="dyn_coat_02_L_01"))
   n24 = &RigUnit_GetTransform(Item=(Type=Bone,Name="pelvis"))
   > &RigUnit_ModifyTransforms(
    ItemToModify="((Item=(Type=Bone,Name=\"dyn_coat_02_L_mstr\"),Transform=(Rotation=(X=0,Y=0,Z=0,W=1),Translation=(X=9.5,Y=0,Z=0),Scale3D=(X=1,Y=1,Z=1))),(Item=(Type=Bone,Name=\"dyn_coat_03_L_mstr\"),Transform=(Rotation=(X=0,Y=0,Z=0,W=1),Translation=(X=5,Y=0,Z=0),Scale3D=(X=1, ...(truncated)"
    Weight=&RigVMFunction_MathFloatRemap(
     SourceMaximum=10
     SourceMinimum=15
     TargetMinimum=0
     bClamp=true
     Value=&RigVMFunction_MathVectorDistance(B=@n23.Transform.Translation, A=@n24.Transform.Translation).Result
    ).Result
   )
   n25 = &RigUnit_SphericalPoseReader(
    ActiveRegionScaleFactors=(PositiveWidth=0.3,NegativeWidth=0.3,PositiveHeight=0.3,NegativeHeight=0.3)
    ActiveRegionSize=0.1
    Debug=(bDrawDebug=true,DebugScale=25,DebugSegments=20,DebugThickness=0.25)
    DriverAxis=(X=1,Y=0,Z=0)
    DriverItem=(Type=Bone,Name="dyn_coat_01_R_01")
    ExecutePin=()
    FalloffRegionScaleFactors=(PositiveWidth=1.5,NegativeWidth=1.5,PositiveHeight=1.5,NegativeHeight=1.5)
    FalloffSize=0.2
    OptionalParentItem=(Type=Bone,Name="pelvis")
    RotationOffset=(X=70,Y=-30,Z=0)
   )
   > &RigUnit_ModifyTransforms(
    ItemToModify="((Item=(Type=Bone,Name=\"dyn_coat_01_R_mstr\"),Transform=(Rotation=(X=0,Y=0,Z=0,W=1),Translation=(X=2,Y=0,Z=-4),Scale3D=(X=1,Y=1,Z=1))),(Item=(Type=Bone,Name=\"dyn_coat_02_R_mstr\"),Transform=(Rotation=(X=0,Y=0,Z=0,W=1),Translation=(X=-5,Y=0,Z=0),Scale3D=(X=1, ...(truncated)"
    Weight=@n25.OutputParam
   )
   n26 = &RigUnit_GetRelativeTransformForItem(Child=(Type=Bone,Name="dyn_chain"), Parent=(Type=Bone,Name="spine_04"), bChildInitial=false, bParentInitial=false)
   &RigVMFunction_MathFloatRemap(
    SourceMaximum=-20
    SourceMinimum=0
    TargetMinimum=0
    bClamp=true
    Value=&RigVMFunction_MathFloatAdd(B=0, A=&RigVMFunction_MathQuaternionToEuler(RotationOrder=ZYX, Value=@n26.RelativeTransform.Rotation).Result.Y) {template="Add::Execute(in A,in B,out Result)"}.Result
   )
   n27 = &RigUnit_GetRelativeTransformForItem(Child=(Type=Bone,Name="dyn_coat_01_R_01"), Parent=(Type=Bone,Name="dyn_coat_01_R_mstr"), bChildInitial=false, bParentInitial=false)
   &RigVMFunction_MathFloatRemap(
    SourceMaximum=-20
    SourceMinimum=-60
    TargetMinimum=0
    bClamp=true
    Value=&RigVMFunction_MathFloatAdd(B=0, A=&RigVMFunction_MathQuaternionToEuler(RotationOrder=ZYX, Value=@n27.RelativeTransform.Rotation).Result.Z) {template="Add::Execute(in A,in B,out Result)"}.Result
   )
   &RigVMFunction_MathQuaternionFromEuler(Euler=(X=0,Y=0,Z=0))
   n28 = &RigVMFunction_MathQuaternionToEuler(
    RotationOrder=ZYX
    Value=&RigUnit_GetRelativeTransformForItem(Child=(Type=Bone,Name="head"), Parent=(Type=Bone,Name="spine_05"), bChildInitial=false, bParentInitial=false).RelativeTransform.Rotation
   )
   &RigUnit_ModifyTransforms(
    ExecutePin=()
    ItemToModify=((Item=(Type=Bone,Name="dyn_lapel_l"),Transform=(Rotation=(X=0,Y=0,Z=0,W=1),Translation=(X=9,Y=-6,Z=0),Scale3D=(X=1,Y=1,Z=1))))
    Weight=&RigVMFunction_MathFloatRemap(
     SourceMaximum=-25
     SourceMinimum=0
     TargetMinimum=0
     bClamp=true
     Value=&RigVMFunction_MathFloatAdd(B=0, A=@n28.Result.Z) {template="Add::Execute(in A,in B,out Result)"}.Result
    ).Result
   )
  ]}
]
hierarchy [
 root {loc=<0 0 0>, rot=<0 0 0 1>}
 pelvis {parent=root, loc=<0 -0.7071 93.1191>, rot=<0.0127 0.707 -0.0127 -0.707>}
 spine_01 {parent=pelvis, loc=<6.7494 -0.6388 0>, rot=<0 0 0.034 -0.9994>}
 spine_02 {parent=spine_01, loc=<7.2518 -0.035 0>, rot=<0 0 0.0238 -0.9997>}
 spine_03 {parent=spine_02, loc=<6.0104 0 0>, rot=<0 0 0.0149 -0.9999>}
 spine_04 {parent=spine_03, loc=<8.1024 0 0>, rot=<0 0 -0.0132 -0.9999>}
 spine_05 {parent=spine_04, loc=<10.0146 0 0>, rot=<0 0 -0.0015 -1>}
 attach_cape {parent=spine_05, loc=<9.017 -11.714 0>, rot=<0.0616 -0.7044 -0.0616 -0.7044>}
 neck_01 {parent=spine_05, loc=<11.3138 0.1529 0>, rot=<0 0 -0.2003 -0.9797>}
 neck_02 {parent=neck_01, loc=<4.5627 0 0>, rot=<0 0 0.0504 -0.9987>}
 head {parent=neck_02, loc=<5.3914 0 0>, rot=<0 0 0.115 -0.9934>}
 attach_backpack {parent=spine_05, loc=<-0.123 -13.473 0>, rot=<-0.0093 -0.707 0.0093 -0.707>}
 clavicle_l {parent=spine_05, loc=<4.3575 -0.1466 -2.3223>, rot=<0.0465 -0.7289 -0.009 -0.6829>}
 upperarm_l {parent=clavicle_l, loc=<15.5111 0 0>, rot=<0.0226 -0.3234 -0.0088 -0.9459>}
 lowerarm_l {parent=upperarm_l, loc=<26.6514 0 0>, rot=<0.0391 -0.0099 -0.2453 -0.9686>}
 hand_l {parent=lowerarm_l, loc=<22.3491 0 0>, rot=<-0.7081 -0.0507 0.0975 -0.6975>}
 thumb_01_l {parent=hand_l, loc=<1.3118 -0.3191 -2.459>, rot=<0.6546 -0.4042 0.0664 -0.6354>}
 thumb_02_l {parent=thumb_01_l, loc=<5.9336 -0.0448 0.0401>, rot=<-0.0037 0.0213 0.298 -0.9543>}
 thumb_03_l {parent=thumb_02_l, loc=<2.2223 0.3682 -0.0907>, rot=<0.0118 -0.013 0.0066 -0.9998>}
 index_metacarpal_l {parent=hand_l, loc=<1.6535 0.0183 -2.96>, rot=<0.0158 -0.0124 -0.0173 -0.9996>}
 index_01_l {parent=index_metacarpal_l, loc=<8.3008 0.0357 -0.0524>, rot=<-0.0044 0.0485 0.0924 -0.9945>}
 index_02_l {parent=index_01_l, loc=<3.3587 -0.3004 0.0452>, rot=<-0.002 0.0027 0.3225 -0.9466>}
 index_03_l {parent=index_02_l, loc=<2.5578 -0.123 -0.0252>, rot=<-0.002 0.0067 0.2558 -0.9667>}
 middle_metacarpal_l {parent=hand_l, loc=<2.1856 -0.1725 -0.5984>, rot=<0.006 0.0168 -0.0329 -0.9993>}
 middle_01_l {parent=middle_metacarpal_l, loc=<7.3275 -0.2447 0.0129>, rot=<-0.0208 0.0042 0.0858 -0.9961>}
 middle_02_l {parent=middle_01_l, loc=<4.0123 0.1401 0.1175>, rot=<-0.0218 0.0258 0.3512 -0.9357>}
 middle_03_l {parent=middle_02_l, loc=<2.6077 0.0534 0.1373>, rot=<0.0082 0.0188 0.1708 -0.9851>}
 ring_metacarpal_l {parent=hand_l, loc=<2.2074 -0.3625 1.3638>, rot=<-0.0284 0.0722 -0.0458 -0.9959>}
 ring_01_l {parent=ring_metacarpal_l, loc=<7.0295 -0.6197 -0.1446>, rot=<-0.0489 -0.0252 0.1892 -0.9804>}
 ring_02_l {parent=ring_01_l, loc=<3.6174 0.4647 -0.1027>, rot=<0 0 0.2806 -0.9598>}
 ring_03_l {parent=ring_02_l, loc=<2.4152 0.114 0.0199>, rot=<0 0 0.1189 -0.9929>}
 pinky_metacarpal_l {parent=hand_l, loc=<1.7616 -0.9753 3.3814>, rot=<-0.0397 0.0814 -0.069 -0.9935>}
 pinky_01_l {parent=pinky_metacarpal_l, loc=<7.1709 -0.4831 -0.1094>, rot=<-0.0654 -0.0035 0.2242 -0.9723>}
 pinky_02_l {parent=pinky_01_l, loc=<2.8274 0.0893 -0.1043>, rot=<0.0037 0.0011 0.2854 -0.9584>}
 pinky_03_l {parent=pinky_02_l, loc=<2.2046 0.0307 -0.0078>, rot=<-0.0002 0.0098 0.1445 -0.9895>}
 lowerarm_twist_02_l {parent=lowerarm_l, loc=<9.6199 0 0>, rot=<0 0 0 -1>}
 lowerarm_twist_01_l {parent=lowerarm_l, loc=<17.8549 0 0>, rot=<0 0 0 -1>}
 upperarm_twist_01_l {parent=upperarm_l, loc=<0.0041 0 0>, rot=<0 0 0 1>}
 upperarm_twist_02_l {parent=upperarm_l, loc=<13.7941 0 0>, rot=<0 0 0 1>}
 clavicle_r {parent=spine_05, loc=<4.3575 -0.1466 2.3223>, rot=<0.7289 0.0465 -0.6829 0.009>}
 upperarm_r {parent=clavicle_r, loc=<-15.5111 0 0>, rot=<0.0226 -0.3234 -0.0088 -0.9459>}
 lowerarm_r {parent=upperarm_r, loc=<-26.6514 0 0>, rot=<0.0391 -0.0099 -0.2453 -0.9686>}
 hand_r {parent=lowerarm_r, loc=<-22.3491 0 0>, rot=<-0.7081 -0.0507 0.0975 -0.6975>}
 thumb_01_r {parent=hand_r, loc=<-1.3118 0.3191 2.459>, rot=<0.6546 -0.4042 0.0664 -0.6354>}
 thumb_02_r {parent=thumb_01_r, loc=<-5.9336 0.0448 -0.0401>, rot=<-0.0037 0.0213 0.298 -0.9543>}
 thumb_03_r {parent=thumb_02_r, loc=<-2.2223 -0.3682 0.0907>, rot=<0.0118 -0.013 0.0066 -0.9998>}
 index_metacarpal_r {parent=hand_r, loc=<-1.6535 -0.0183 2.96>, rot=<0.0158 -0.0124 -0.0173 -0.9996>}
 index_01_r {parent=index_metacarpal_r, loc=<-8.3008 -0.0357 0.0524>, rot=<-0.0044 0.0485 0.0924 -0.9945>}
 index_02_r {parent=index_01_r, loc=<-3.3587 0.3004 -0.0452>, rot=<-0.002 0.0027 0.3225 -0.9466>}
 index_03_r {parent=index_02_r, loc=<-2.5578 0.123 0.0252>, rot=<-0.002 0.0067 0.2558 -0.9667>}
 middle_metacarpal_r {parent=hand_r, loc=<-2.1856 0.1725 0.5984>, rot=<0.006 0.0168 -0.0329 -0.9993>}
 middle_01_r {parent=middle_metacarpal_r, loc=<-7.3275 0.2447 -0.0129>, rot=<-0.0208 0.0042 0.0858 -0.9961>}
 middle_02_r {parent=middle_01_r, loc=<-4.0123 -0.1401 -0.1175>, rot=<-0.0218 0.0258 0.3512 -0.9357>}
 middle_03_r {parent=middle_02_r, loc=<-2.6077 -0.0534 -0.1373>, rot=<0.0082 0.0188 0.1708 -0.9851>}
 ring_metacarpal_r {parent=hand_r, loc=<-2.2074 0.3625 -1.3638>, rot=<-0.0284 0.0722 -0.0458 -0.9959>}
 ring_01_r {parent=ring_metacarpal_r, loc=<-7.0295 0.6197 0.1446>, rot=<-0.0489 -0.0252 0.1892 -0.9804>}
 ring_02_r {parent=ring_01_r, loc=<-3.6174 -0.4647 0.1027>, rot=<0 0 0.2806 -0.9598>}
 ring_03_r {parent=ring_02_r, loc=<-2.4152 -0.114 -0.0199>, rot=<0 0 0.1189 -0.9929>}
 pinky_metacarpal_r {parent=hand_r, loc=<-1.7616 0.9753 -3.3814>, rot=<-0.0397 0.0814 -0.069 -0.9935>}
 pinky_01_r {parent=pinky_metacarpal_r, loc=<-7.1709 0.4831 0.1094>, rot=<-0.0654 -0.0035 0.2242 -0.9723>}
 pinky_02_r {parent=pinky_01_r, loc=<-2.8274 -0.0893 0.1043>, rot=<0.0037 0.0011 0.2854 -0.9584>}
 pinky_03_r {parent=pinky_02_r, loc=<-2.2046 -0.0307 0.0078>, rot=<-0.0002 0.0098 0.1445 -0.9895>}
 lowerarm_twist_02_r {parent=lowerarm_r, loc=<-9.6199 0 0>, rot=<0 0 0 -1>}
 lowerarm_twist_01_r {parent=lowerarm_r, loc=<-17.8549 0 0>, rot=<0 0 0 -1>}
 upperarm_twist_01_r {parent=upperarm_r, loc=<-0.0041 0 0>, rot=<0 0 0 -1>}
 upperarm_twist_02_r {parent=upperarm_r, loc=<-13.7941 0 0>, rot=<0 0 0 -1>}
 dyn_collar_l {parent=spine_05, loc=<11.0489 5.5961 -6.4912>, rot=<0.1731 -0.6856 -0.1731 -0.6856>}
 dyn_collar_b {parent=spine_05, loc=<13.7167 -2.7111 0>, rot=<0.5196 -0.4795 0.4795 -0.5196>}
 dyn_collar_r {parent=spine_05, loc=<11.0491 5.5961 6.4912>, rot=<0.6856 0.1731 -0.6856 0.1731>}
 dyn_chain_mstr {parent=spine_04, loc=<9.6002 12.4652 0>, rot=<0.0294 -0.7065 -0.0294 -0.7065>}
 dyn_chain {parent=dyn_chain_mstr, loc=<0 0 0>, rot=<-0.1333 -0.6944 -0.1333 -0.6944>}
 dyn_medal_mst {parent=dyn_chain, loc=<6.6446 0 0>, rot=<0 0 0 1>}
 dyn_medal {parent=dyn_medal_mst, loc=<0 0 0>, rot=<0 0 0 1>}
 medal loc {controlType=EULER_TRANSFORM, parent=dyn_medal, type=Control, loc=<0 0 0>, rot=<0 0 0 1>}
 dyn_lapel_l {parent=spine_03, loc=<5.7325 13.3111 -6.8244>, rot=<0.4719 0.5266 0.5266 0.4719>}
 dyn_rapel_r {parent=spine_03, loc=<5.7321 13.3111 6.8244>, rot=<0.4719 0.5266 0.5266 0.4719>}
 thigh_l {parent=pelvis, loc=<-2.5693 -0.0109 -7.7501>, rot=<0.0012 0.0572 0.0155 -0.9982>}
 calf_l {parent=thigh_l, loc=<-39.0484 0 0>, rot=<0.0006 -0.0161 -0.0391 -0.9991>}
 foot_l {parent=calf_l, loc=<-42.561 0.1077 0.4292>, rot=<0.0418 -0.0455 0.0343 -0.9975>}
 ball_l {parent=foot_l, loc=<-6.1599 13.8264 0.6806>, rot=<0.0001 -0.0001 -0.7063 -0.7079>}
 calf_twist_02_l {parent=calf_l, loc=<-18.8064 -0.0967 0.0397>, rot=<0.0237 0.0095 -0.0035 -0.9997>}
 calf_twist_01_l {parent=calf_l, loc=<-35.813 -0.0967 0.0397>, rot=<0.0237 0.0095 -0.0035 -0.9997>}
 thigh_twist_01_l {parent=thigh_l, loc=<-21.6303 0 0>, rot=<0.0013 0 0 -1>}
 thigh_r {parent=pelvis, loc=<-2.5693 -0.0109 7.7501>, rot=<0.0572 -0.0012 0.9982 0.0155>}
 calf_r {parent=thigh_r, loc=<39.0484 0 0>, rot=<0.0006 -0.0161 -0.0391 -0.9991>}
 foot_r {parent=calf_r, loc=<42.561 -0.1077 -0.4292>, rot=<0.0418 -0.0455 0.0343 -0.9975>}
 ball_r {parent=foot_r, loc=<6.1599 -13.8264 -0.6806>, rot=<0.0001 -0.0001 -0.7063 -0.7079>}
 calf_twist_02_r {parent=calf_r, loc=<18.8064 0.0967 -0.0397>, rot=<0.0237 0.0095 -0.0035 -0.9997>}
 calf_twist_01_r {parent=calf_r, loc=<35.813 0.0967 -0.0397>, rot=<0.0237 0.0095 -0.0035 -0.9997>}
 thigh_twist_01_r {parent=thigh_r, loc=<21.6303 0 0>, rot=<0.0013 0 0 -1>}
 dyn_coat_01_L_mstr {parent=pelvis, loc=<15.4447 8.905 -9.1055>, rot=<-0.0127 -0.707 0.0127 -0.707>}
 dyn_coat_01_L_01 {parent=dyn_coat_01_L_mstr, loc=<0 0 0>, rot=<0.3102 -0.6285 -0.3159 -0.6395>}
 dyn_coat_01_L_02 {parent=dyn_coat_01_L_01, loc=<14.1478 0 0>, rot=<0 0 0 1>}
 dyn_coat_01_L_03 {parent=dyn_coat_01_L_02, loc=<29.1839 0 0>, rot=<0 0 0 1>}
 dyn_coat_01_L_04 {parent=dyn_coat_01_L_03, loc=<23.015 0 0>, rot=<0 0 0 1>}
 dyn_coat_02_L_mstr {parent=pelvis, loc=<5.2593 0.6084 -15.1957>, rot=<-0.0127 -0.707 0.0127 -0.707>}
 dyn_coat_02_L_01 {parent=dyn_coat_02_L_mstr, loc=<0 0 0>, rot=<0 -0.6514 0 -0.7587>}
 dyn_coat_02_L_02 {parent=dyn_coat_02_L_01, loc=<17.6512 0 0>, rot=<0 0 0 1>}
 dyn_coat_02_L_03 {parent=dyn_coat_02_L_02, loc=<18.7727 0 0>, rot=<0 0 0 1>}
 dyn_coat_02_L_04 {parent=dyn_coat_02_L_03, loc=<19.5611 0 0>, rot=<0 0 0 1>}
 dyn_coat_03_L_mstr {parent=pelvis, loc=<-2.4831 -10.6382 -10.3225>, rot=<0.3279 -0.6265 0.3502 -0.6143>}
 dyn_coat_03_L_01 {parent=dyn_coat_03_L_mstr, loc=<0 0 0>, rot=<-0.0073 -0.6747 -0.0076 -0.738>}
 dyn_coat_03_L_02 {parent=dyn_coat_03_L_01, loc=<15.308 0 0>, rot=<0 0 0 1>}
 dyn_coat_03_L_03 {parent=dyn_coat_03_L_02, loc=<16.4296 0 0>, rot=<0 0 0 1>}
 dyn_coat_03_L_04 {parent=dyn_coat_03_L_03, loc=<17.2179 0 0>, rot=<0 0 0 1>}
 dyn_coat_B_mstr {parent=pelvis, loc=<-2.3436 -13.648 0>, rot=<-0.0127 -0.707 0.0127 -0.707>}
 dyn_coat_B_01 {parent=dyn_coat_B_mstr, loc=<0 0 0>, rot=<-0.4659 -0.4659 0.5319 -0.5319>}
 dyn_coat_B_02 {parent=dyn_coat_B_01, loc=<15.133 0 0>, rot=<0 0 0 1>}
 dyn_coat_B_03 {parent=dyn_coat_B_02, loc=<16.2546 0 0>, rot=<0 0 0 1>}
 dyn_coat_B_04 {parent=dyn_coat_B_03, loc=<17.0429 0 0>, rot=<0 0 0 1>}
 dyn_coat_01_R_mstr {parent=pelvis, loc=<15.4448 8.905 9.1055>, rot=<0.707 -0.0127 0.707 0.0127>}
 dyn_coat_01_R_01 {parent=dyn_coat_01_R_mstr, loc=<0 0 0>, rot=<-0.2973 -0.6252 0.3183 -0.6476>}
 dyn_coat_01_R_02 {parent=dyn_coat_01_R_01, loc=<21.1051 0 0>, rot=<0 0 0 1>}
 dyn_coat_01_R_03 {parent=dyn_coat_01_R_02, loc=<22.2266 0 0>, rot=<0 0 0 1>}
 dyn_coat_01_R_04 {parent=dyn_coat_01_R_03, loc=<23.015 0 0>, rot=<0 0 0 1>}
 dyn_coat_02_R_mstr {parent=pelvis, loc=<5.2592 0.6084 15.1957>, rot=<0.707 -0.0127 -0.707 -0.0127>}
 dyn_coat_02_R_01 {parent=dyn_coat_02_R_mstr, loc=<0 0 0>, rot=<0 0.7587 0 -0.6514>}
 dyn_coat_02_R_02 {parent=dyn_coat_02_R_01, loc=<17.649 0 0.2765>, rot=<0 0.0078 0 -1>}
 dyn_coat_02_R_03 {parent=dyn_coat_02_R_02, loc=<18.7727 0 -0.0001>, rot=<0 0 0 1>}
 dyn_coat_02_R_04 {parent=dyn_coat_02_R_03, loc=<19.5612 0 0.0001>, rot=<0 0 0 1>}
 dyn_coat_03_R_mstr {parent=pelvis, loc=<-2.4831 -10.6381 10.3225>, rot=<0.6265 0.3279 -0.6143 -0.3502>}
 dyn_coat_03_R_01 {parent=dyn_coat_03_R_mstr, loc=<0 0 0>, rot=<-0.0076 0.738 0.0073 -0.6747>}
 dyn_coat_03_R_02 {parent=dyn_coat_03_R_01, loc=<15.3081 0 0>, rot=<0 0 0 -1>}
 dyn_coat_03_R_03 {parent=dyn_coat_03_R_02, loc=<16.4295 0 0>, rot=<0 0 0 -1>}
 dyn_coat_03_R_04 {parent=dyn_coat_03_R_03, loc=<17.2179 0 0>, rot=<0 0 0 -1>}
 dyn_button_01_r {parent=pelvis, loc=<10.0822 5.1985 13.8066>, rot=<0.2558 0.6447 -0.218 -0.6866>, scale=<-1 -1 -1>}
 dyn_button_02_r {parent=pelvis, loc=<-2.8366 6.3676 14.6612>, rot=<0.281 0.6155 -0.2313 -0.6991>, scale=<-1 -1 -1>}
 dyn_button_02_l {parent=pelvis, loc=<-2.8366 6.3676 -14.6612>, rot=<-0.281 -0.6155 -0.2313 -0.6991>}
 dyn_button_01_l {parent=pelvis, loc=<10.0822 5.1985 -13.8066>, rot=<-0.2558 -0.6447 -0.218 -0.6866>}
]
