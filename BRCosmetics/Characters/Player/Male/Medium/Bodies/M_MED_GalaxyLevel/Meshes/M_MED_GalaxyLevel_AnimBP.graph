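# Animation Blueprint for the GalaxyLevel medium male cosmetic body, adding secondary motion to coat, necklace and buttons.
# Copies the attached parent's pose through two Control Rigs into a cache, then bRigidBodyEnabled selects physics or a fallback of leg-driven constraints and trail controllers.
# The physics path blends state-specific coat RigidBody assets, then adds necklace simulation and button constraints.
# The EventGraph derives movement and thigh alphas, switches sliding physics and resets dynamics after slides.
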AnimBlueprint /BRCosmetics/Characters/Player/Male/Medium/Bodies/M_MED_GalaxyLevel/Meshes/M_MED_GalaxyLevel_AnimBP
target_skeleton=&/M_MED_GalaxyLevel_Skeleton
graphs [
 AnimGraph {nodes [
   Control Rig!(
    Source
    Alpha
    DinoHead:bool=false
    IsSlidingCtrl:bool=false
    IsCrouchingCtrl:bool=false
    IsGlidingCtrl:bool=Get bIsSkyDiving()
    IsScaleCtrl:bool=false
    RbanOff:bool=true
   ) -> (Pose) {
    alphaBoolBlend {blendOption=Linear}
    controlRigAssetReference {blueprintRigClass=&/M_MED_GalaxyLevel_CtrlRig.M_MED_GalaxyLevel_CtrlRig_C}
   }
   Control Rig(DinoHead=false, IsSlidingCtrl=false, IsCrouchingCtrl=false, IsGlidingCtrl=Get bIsSkyDiving(), IsScaleCtrl=false, RbanOff=true) {
    alphaBoolBlend {blendOption=Linear}
    controlRigAssetReference {blueprintRigClass=&M_MED_GalaxyLevel_CtrlRig_C}
   }
   "Transform (Modify) Bone - Bone: dyn_coat_03_R_01"(Alpha=0, Translation="-0.793,0.116,2.958", Rotation="0,0,0", Scale="1,1,1") {
    rotationMode=BMM_Additive
    rotationSpace=BCS_BoneSpace
    translationMode=BMM_Additive
    translationSpace=BCS_BoneSpace
    alphaBoolBlend {blendOption=Linear}
    boneToModify {boneName=dyn_coat_03_R_01}
    translation=<-0.793 0.116 2.958>
   }
   n1 = Get FwdMovement() -> (FwdMovement:double)
   n2 = "Transform (Modify) Bone - Bone: dyn_coat_03_L_mstr"(Alpha=@n1, Translation="0,0,0", Rotation="0,0,0", Scale="1,1,1") {
    rotationMode=BMM_Additive
    rotationSpace=BCS_BoneSpace
    translationMode=BMM_Additive
    translationSpace=BCS_BoneSpace
    alphaBoolBlend {blendOption=Linear}
    boneToModify {boneName=dyn_coat_03_L_mstr}
   }
   "Transform (Modify) Bone - Bone: dyn_coat_03_L_mstr"(ComponentPose=@n2, Alpha=Get thigh_rot_otr_l() -> (thigh_rot_otr_l:double), Translation="0,0,0", Rotation="35,0,0", Scale="1,1,1") {
    rotationMode=BMM_Additive
    rotationSpace=BCS_BoneSpace
    translationMode=BMM_Additive
    translationSpace=BCS_BoneSpace
    alphaBoolBlend {blendOption=Linear}
    boneToModify {boneName=dyn_coat_03_L_mstr}
    rotation {pitch=35}
   }
   n3 = "Transform (Modify) Bone - Bone: dyn_coat_B_mstr"(Alpha=Get FwdMovement(), Translation="0,0,0", Rotation="0,0,30", Scale="1,1,1") {
    rotationMode=BMM_Additive
    rotationSpace=BCS_BoneSpace
    translationMode=BMM_Additive
    translationSpace=BCS_BoneSpace
    alphaBoolBlend {blendOption=Linear}
    boneToModify {boneName=dyn_coat_B_mstr}
    rotation {roll=30}
   }
   n4 = "Constraint - dyn_coat_03_L_01 (PPP)"!(ComponentPose=@n3, Alpha=1, ConstraintWeights_0=0.5, ConstraintWeights_1:float=0, ConstraintWeights_2:float=0.5) -> (Pose) {
    alphaBoolBlend {blendOption=Linear}
    boneToModify {boneName=dyn_coat_03_L_01}
    constraintSetup [{targetBone {boneName=dyn_coat_B_01}}, {targetBone {boneName=dyn_coat_B_01}}, {targetBone {boneName=dyn_coat_02_L_01}}]
    constraintWeights [1,1,1]
   }
   n5 = "Constraint - dyn_coat_03_L_02 (PPP)"!(ComponentPose=@n4, Alpha=1, ConstraintWeights_0=0.5, ConstraintWeights_1:float=0, ConstraintWeights_2:float=0.5) -> (Pose) {
    alphaBoolBlend {blendOption=Linear}
    boneToModify {boneName=dyn_coat_03_L_02}
    constraintSetup [{targetBone {boneName=dyn_coat_B_02}}, {targetBone {boneName=dyn_coat_B_01}}, {targetBone {boneName=dyn_coat_02_L_02}}]
    constraintWeights [1,1,1]
   }
   n6 = "Constraint - dyn_coat_03_R_01 (PPP)"!(ComponentPose=@n5, Alpha=1, ConstraintWeights_0=0.8, ConstraintWeights_1:float=0.2, ConstraintWeights_2:float=1) -> (Pose) {
    alphaBoolBlend {blendOption=Linear}
    boneToModify {boneName=dyn_coat_03_R_01}
    constraintSetup [{targetBone {boneName=dyn_coat_B_01}}, {targetBone {boneName=dyn_coat_B_01}}, {targetBone {boneName=thigh_r}}]
    constraintWeights [1,1,1]
   }
   n7 = "Transform (Modify) Bone - Bone: dyn_coat_03_R_mstr"(ComponentPose=@n6, Alpha=@n1, Translation="0,0,0", Rotation="55,10,0", Scale="1,1,1") {
    rotationMode=BMM_Additive
    rotationSpace=BCS_BoneSpace
    translationMode=BMM_Additive
    translationSpace=BCS_BoneSpace
    alphaBoolBlend {blendOption=Linear}
    boneToModify {boneName=dyn_coat_03_R_mstr}
    rotation {pitch=55, yaw=10}
   }
   n8 = "Transform (Modify) Bone - Bone: dyn_coat_03_R_mstr"(ComponentPose=@n7, Alpha=Get thigh_rot_otr_r() -> (thigh_rot_otr_r:double), Translation="0,0,0", Rotation="45,15,0", Scale="1,1,1") {
    rotationMode=BMM_Additive
    rotationSpace=BCS_BoneSpace
    translationMode=BMM_Additive
    translationSpace=BCS_BoneSpace
    alphaBoolBlend {blendOption=Linear}
    boneToModify {boneName=dyn_coat_03_R_mstr}
    rotation {pitch=45, yaw=15}
   }
   n9 = "Transform (Modify) Bone - Bone: dyn_coat_03_R_mstr"(ComponentPose=@n8, Alpha=0, Translation="0,0,0", Rotation="15,15,0", Scale="1,1,1") {
    rotationMode=BMM_Additive
    rotationSpace=BCS_BoneSpace
    translationMode=BMM_Additive
    translationSpace=BCS_BoneSpace
    alphaBoolBlend {blendOption=Linear}
    boneToModify {boneName=dyn_coat_03_R_mstr}
    rotation {pitch=15, yaw=15}
   }
   "Transform (Modify) Bone - Bone: dyn_coat_03_L_mstr"(ComponentPose=@n9, Alpha=0, Translation="0,0,0", Rotation="15,15,0", Scale="1,1,1") {
    rotationMode=BMM_Additive
    rotationSpace=BCS_BoneSpace
    translationMode=BMM_Additive
    translationSpace=BCS_BoneSpace
    alphaBoolBlend {blendOption=Linear}
    boneToModify {boneName=dyn_coat_03_L_mstr}
    rotation {pitch=15, yaw=15}
   }
   RigidBody!(ComponentPose, Alpha, OverrideWorldGravity:Vector="0,0,0", ExternalForce) -> (Pose) {
    bClampLinearTranslationLimitToRefPose=true
    bOverrideWorldGravity=true
    overridePhysicsAsset=/Game/Characters/Player/Male/Medium/Bodies/M_MED_GalaxyLevel/Meshes/M_MED_GalaxyLevel_crouch_Physics
    simulationSpace=BaseBoneSpace
    alphaBoolBlend {blendOption=Linear}
    baseBoneRef {boneName=pelvis}
    componentAppliedLinearAccClamp=<3000 3000 3000>
    componentLinearAccScale=<0.15 0.15 0.2>
    componentLinearVelScale=<0.15 0.15 0.2>
    simSpaceSettings {maxAngularVelocity=2, maxLinearAcceleration=700, maxLinearVelocity=500, velocityScaleZ=0.1, worldAlpha=0.6}
   }
   n10 = "Constraint - dyn_coat_01_L_02 (P)"(ComponentPose=Local To Component(LocalPose=Use cached pose 'CtrlRigCache'()), Alpha=1, ConstraintWeights_0=1) {
    alphaBoolBlend {blendOption=Linear}
    boneToModify {boneName=dyn_coat_01_L_02}
    constraintSetup [{targetBone {boneName=thigh_l}}]
    constraintWeights [1]
   }
   n11 = "Constraint - dyn_coat_01_L_03 (P)"(ComponentPose=@n10, Alpha=1, ConstraintWeights_0=1) {
    alphaBoolBlend {blendOption=Linear}
    boneToModify {boneName=dyn_coat_01_L_03}
    constraintSetup [{targetBone {boneName=calf_l}}]
    constraintWeights [1]
   }
   n12 = "Constraint - dyn_coat_01_R_02 (P)"(ComponentPose=@n11, Alpha=1, ConstraintWeights_0=1) {
    alphaBoolBlend {blendOption=Linear}
    boneToModify {boneName=dyn_coat_01_R_02}
    constraintSetup [{targetBone {boneName=thigh_r}}]
    constraintWeights [1]
   }
   n13 = "Constraint - dyn_coat_01_R_03 (P)"(ComponentPose=@n12, Alpha=1, ConstraintWeights_0=1) {
    alphaBoolBlend {blendOption=Linear}
    boneToModify {boneName=dyn_coat_01_R_03}
    constraintSetup [{targetBone {boneName=calf_r}}]
    constraintWeights [1]
   }
   n14 = "Constraint - dyn_coat_03_L_01 (R)"(ComponentPose=@n13, Alpha=1, ConstraintWeights_0=1) {
    alphaBoolBlend {blendOption=Linear}
    boneToModify {boneName=dyn_coat_03_L_01}
    constraintSetup [{transformType=Rotation, targetBone {boneName=thigh_l}}]
    constraintWeights [1]
   }
   n15 = "Constraint - dyn_coat_03_L_02 (RR)"!(ComponentPose=@n14, Alpha=1, ConstraintWeights_0=0, ConstraintWeights_1:float=1) -> (Pose) {
    alphaBoolBlend {blendOption=Linear}
    boneToModify {boneName=dyn_coat_03_L_02}
    constraintSetup [{transformType=Rotation, targetBone {boneName=thigh_l}}, {transformType=Rotation, targetBone {boneName=calf_l}}]
    constraintWeights [1,1]
   }
   n16 = "Constraint - dyn_coat_03_L_03 (RR)"!(ComponentPose=@n15, Alpha=1, ConstraintWeights_0=1, ConstraintWeights_1:float=0) -> (Pose) {
    alphaBoolBlend {blendOption=Linear}
    boneToModify {boneName=dyn_coat_03_L_03}
    constraintSetup [{transformType=Rotation, targetBone {boneName=calf_l}}, {transformType=Rotation, targetBone {boneName=thigh_l}}]
    constraintWeights [1,1]
   }
   n17 = "Constraint - dyn_coat_03_R_01 (R)"(ComponentPose=@n16, Alpha=1, ConstraintWeights_0=1) {
    alphaBoolBlend {blendOption=Linear}
    boneToModify {boneName=dyn_coat_03_R_01}
    constraintSetup [{transformType=Rotation, targetBone {boneName=thigh_r}}]
    constraintWeights [1]
   }
   n18 = "Constraint - dyn_coat_03_R_02 (RR)"!(ComponentPose=@n17, Alpha=1, ConstraintWeights_0=1, ConstraintWeights_1:float=0) -> (Pose) {
    alphaBoolBlend {blendOption=Linear}
    boneToModify {boneName=dyn_coat_03_R_02}
    constraintSetup [{transformType=Rotation, targetBone {boneName=thigh_r}}, {transformType=Rotation, targetBone {boneName=calf_r}}]
    constraintWeights [1,1]
   }
   n19 = "Constraint - dyn_coat_03_R_03 (RR)"!(ComponentPose=@n18, Alpha=1, ConstraintWeights_0=1, ConstraintWeights_1:float=0) -> (Pose) {
    alphaBoolBlend {blendOption=Linear}
    boneToModify {boneName=dyn_coat_03_R_03}
    constraintSetup [{transformType=Rotation, targetBone {boneName=calf_r}}, {transformType=Rotation, targetBone {boneName=thigh_r}}]
    constraintWeights [1,1]
   }
   n20 = "Transform (Modify) Bone - Bone: dyn_coat_03_R_01"(ComponentPose=@n19, Alpha=0, Translation="-0.793,0.116,2.958", Rotation="0,0,0", Scale="1,1,1") {
    rotationMode=BMM_Additive
    rotationSpace=BCS_BoneSpace
    translationMode=BMM_Additive
    translationSpace=BCS_BoneSpace
    alphaBoolBlend {blendOption=Linear}
    boneToModify {boneName=dyn_coat_03_R_01}
    translation=<-0.793 0.116 2.958>
   }
   n21 = "Constraint - dyn_coat_B_01 (PP)"!(ComponentPose=@n20, Alpha=1, ConstraintWeights_0=0.5, ConstraintWeights_1:float=0.5) -> (Pose) {
    alphaBoolBlend {blendOption=Linear}
    boneToModify {boneName=dyn_coat_B_01}
    constraintSetup [{targetBone {boneName=dyn_coat_03_L_01}}, {targetBone {boneName=dyn_coat_03_R_01}}]
    constraintWeights [1,1]
   }
   n22 = "Constraint - dyn_coat_B_02 (PP)"!(ComponentPose=@n21, Alpha=1, ConstraintWeights_0=0.5, ConstraintWeights_1:float=0.5) -> (Pose) {
    alphaBoolBlend {blendOption=Linear}
    boneToModify {boneName=dyn_coat_B_02}
    constraintSetup [{targetBone {boneName=dyn_coat_03_L_02}}, {targetBone {boneName=dyn_coat_03_R_02}}]
    constraintWeights [1,1]
   }
   n23 = "Constraint - dyn_coat_B_03 (PP)"!(ComponentPose=@n22, Alpha=1, ConstraintWeights_0=0.5, ConstraintWeights_1:float=0.5) -> (Pose) {
    alphaBoolBlend {blendOption=Linear}
    boneToModify {boneName=dyn_coat_B_03}
    constraintSetup [{targetBone {boneName=dyn_coat_03_L_03}}, {targetBone {boneName=dyn_coat_03_R_03}}]
    constraintWeights [1,1]
   }
   n24 = Component To Local(ComponentPose=@n23)
   n25 = "Constraint - dyn_coat_01_L_02 (P)"(ComponentPose=Local To Component(LocalPose=Use cached pose 'CtrlRigCache'()), Alpha=1, ConstraintWeights_0=1) {
    alphaBoolBlend {blendOption=Linear}
    boneToModify {boneName=dyn_coat_01_L_02}
    constraintSetup [{targetBone {boneName=thigh_l}}]
    constraintWeights [1]
   }
   n26 = "Constraint - dyn_coat_01_L_03 (P)"(ComponentPose=@n25, Alpha=1, ConstraintWeights_0=1) {
    alphaBoolBlend {blendOption=Linear}
    boneToModify {boneName=dyn_coat_01_L_03}
    constraintSetup [{targetBone {boneName=calf_l}}]
    constraintWeights [1]
   }
   n27 = "Constraint - dyn_coat_01_R_02 (P)"(ComponentPose=@n26, Alpha=1, ConstraintWeights_0=1) {
    alphaBoolBlend {blendOption=Linear}
    boneToModify {boneName=dyn_coat_01_R_02}
    constraintSetup [{targetBone {boneName=thigh_r}}]
    constraintWeights [1]
   }
   n28 = "Constraint - dyn_coat_01_R_03 (P)"(ComponentPose=@n27, Alpha=1, ConstraintWeights_0=1) {
    alphaBoolBlend {blendOption=Linear}
    boneToModify {boneName=dyn_coat_01_R_03}
    constraintSetup [{targetBone {boneName=calf_r}}]
    constraintWeights [1]
   }
   n29 = "Constraint - dyn_coat_03_L_01 (R)"(ComponentPose=@n28, Alpha=1, ConstraintWeights_0=1) {
    alphaBoolBlend {blendOption=Linear}
    boneToModify {boneName=dyn_coat_03_L_01}
    constraintSetup [{transformType=Rotation, targetBone {boneName=thigh_l}}]
    constraintWeights [1]
   }
   n30 = "Constraint - dyn_coat_03_L_02 (RR)"(ComponentPose=@n29, Alpha=1, ConstraintWeights_0=0, ConstraintWeights_1=1) {
    alphaBoolBlend {blendOption=Linear}
    boneToModify {boneName=dyn_coat_03_L_02}
    constraintSetup [{transformType=Rotation, targetBone {boneName=thigh_l}}, {transformType=Rotation, targetBone {boneName=calf_l}}]
    constraintWeights [1,1]
   }
   n31 = "Constraint - dyn_coat_03_L_03 (RR)"(ComponentPose=@n30, Alpha=1, ConstraintWeights_0=1, ConstraintWeights_1=0) {
    alphaBoolBlend {blendOption=Linear}
    boneToModify {boneName=dyn_coat_03_L_03}
    constraintSetup [{transformType=Rotation, targetBone {boneName=calf_l}}, {transformType=Rotation, targetBone {boneName=thigh_l}}]
    constraintWeights [1,1]
   }
   n32 = "Constraint - dyn_coat_03_R_01 (R)"(ComponentPose=@n31, Alpha=1, ConstraintWeights_0=1) {
    alphaBoolBlend {blendOption=Linear}
    boneToModify {boneName=dyn_coat_03_R_01}
    constraintSetup [{transformType=Rotation, targetBone {boneName=thigh_r}}]
    constraintWeights [1]
   }
   n33 = "Constraint - dyn_coat_03_R_02 (RR)"(ComponentPose=@n32, Alpha=1, ConstraintWeights_0=1, ConstraintWeights_1=0) {
    alphaBoolBlend {blendOption=Linear}
    boneToModify {boneName=dyn_coat_03_R_02}
    constraintSetup [{transformType=Rotation, targetBone {boneName=thigh_r}}, {transformType=Rotation, targetBone {boneName=calf_r}}]
    constraintWeights [1,1]
   }
   n34 = "Constraint - dyn_coat_03_R_03 (RR)"(ComponentPose=@n33, Alpha=1, ConstraintWeights_0=1, ConstraintWeights_1=0) {
    alphaBoolBlend {blendOption=Linear}
    boneToModify {boneName=dyn_coat_03_R_03}
    constraintSetup [{transformType=Rotation, targetBone {boneName=calf_r}}, {transformType=Rotation, targetBone {boneName=thigh_r}}]
    constraintWeights [1,1]
   }
   n35 = "Constraint - dyn_coat_B_01 (PP)"(ComponentPose=@n34, Alpha=1, ConstraintWeights_0=0.5, ConstraintWeights_1=0.5) {
    alphaBoolBlend {blendOption=Linear}
    boneToModify {boneName=dyn_coat_B_01}
    constraintSetup [{targetBone {boneName=dyn_coat_03_L_01}}, {targetBone {boneName=dyn_coat_03_R_01}}]
    constraintWeights [1,1]
   }
   n36 = "Constraint - dyn_coat_B_02 (PP)"(ComponentPose=@n35, Alpha=1, ConstraintWeights_0=0.5, ConstraintWeights_1=0.5) {
    alphaBoolBlend {blendOption=Linear}
    boneToModify {boneName=dyn_coat_B_02}
    constraintSetup [{targetBone {boneName=dyn_coat_03_L_02}}, {targetBone {boneName=dyn_coat_03_R_02}}]
    constraintWeights [1,1]
   }
   n37 = "Constraint - dyn_coat_B_03 (PP)"(ComponentPose=@n36, Alpha=1, ConstraintWeights_0=0.5, ConstraintWeights_1=0.5) {
    alphaBoolBlend {blendOption=Linear}
    boneToModify {boneName=dyn_coat_B_03}
    constraintSetup [{targetBone {boneName=dyn_coat_03_L_03}}, {targetBone {boneName=dyn_coat_03_R_03}}]
    constraintWeights [1,1]
   }
   n38 = Get BwdMovement() -> (BwdMovement:double)
   n39 = Trail controller - Bone: dyn_coat_03_L_04(ComponentPose=@n37, Alpha=@n38) {
    bLimitRotation=true
    bUsePlanarLimit=true
    chainLength=3
    debugLifeTime=0.1
    maxDeltaTime=0.0333
    alphaScaleBiasClamp {bInterpResult=true}
    rotationLimits [{limitMax=<180 180 180>, limitMin=<-180 -180 -45>}, {limitMax=<180 180 180>, limitMin=<-180 -180 -45>}, {limitMax=<180 180 180>}]
    rotationOffsets [<0 0 0>, <0 0 0>, <0 0 0>]
    trailBone {boneName=dyn_coat_03_L_04}
    trailRelaxationSpeed {editorCurveData {keys [<at 0 15 RCIM_Linear>,<at 0.5289 13.1838 RCIM_Linear>,<at 1 6 RCIM_Linear>]}}
   }
   n40 = Trail controller - Bone: dyn_coat_B_04(ComponentPose=@n39, Alpha=@n38) {
    bLimitRotation=true
    bUsePlanarLimit=true
    chainLength=3
    debugLifeTime=0.1
    maxDeltaTime=0.0333
    alphaScaleBiasClamp {bInterpResult=true}
    rotationLimits [{limitMax=<180 180 180>, limitMin=<-180 -180 -45>}, {limitMax=<180 180 180>, limitMin=<-180 -180 -45>}, {limitMax=<180 180 180>}]
    rotationOffsets [<0 0 0>, <0 0 0>, <0 0 0>]
    trailBone {boneName=dyn_coat_B_04}
    trailRelaxationSpeed {editorCurveData {keys [<at 0 15 RCIM_Linear>,<at 0.5289 13.1838 RCIM_Linear>,<at 1 6 RCIM_Linear>]}}
   }
   n41 = Trail controller - Bone: dyn_coat_03_R_04(ComponentPose=@n40, Alpha=@n38) {
    bLimitRotation=true
    bUsePlanarLimit=true
    chainLength=3
    debugLifeTime=0.1
    maxDeltaTime=0.0333
    alphaScaleBiasClamp {bInterpResult=true}
    rotationLimits [{limitMax=<180 180 180>, limitMin=<-180 -180 -45>}, {limitMax=<180 180 180>, limitMin=<-180 -180 -45>}, {limitMax=<180 180 180>}]
    rotationOffsets [<0 0 0>, <0 0 0>, <0 0 0>]
    trailBone {boneName=dyn_coat_03_R_04}
    trailRelaxationSpeed {editorCurveData {keys [<at 0 15 RCIM_Linear>,<at 0.5289 13.1838 RCIM_Linear>,<at 1 6 RCIM_Linear>]}}
   }
   n42 = Component To Local(ComponentPose=@n41)
   n43 = RigidBody(ComponentPose=Local To Component(LocalPose=Use cached pose 'CtrlRigCache'()), OverrideWorldGravity="0,0,-920") {
    bClampLinearTranslationLimitToRefPose=true
    bOverrideWorldGravity=true
    overridePhysicsAsset=&/M_MED_GalaxyLevel_coat_Physics_Emote
    simulationSpace=BaseBoneSpace
    alphaBoolBlend {blendOption=Linear}
    baseBoneRef {boneName=root}
    componentAppliedLinearAccClamp=<3000 3000 1000>
    componentLinearAccScale=<0.1 0.1 0.2>
    componentLinearVelScale=<0.1 0.1 0.2>
    overrideWorldGravity {z=-920}
    simSpaceSettings {worldAlpha=0.4}
   }
   n44 = RigidBody(ComponentPose=Local To Component(LocalPose=Use cached pose 'CtrlRigCache'()), OverrideWorldGravity="0,0,-920") {
    bClampLinearTranslationLimitToRefPose=true
    bOverrideWorldGravity=true
    overridePhysicsAsset=&/M_MED_GalaxyLevel_coat_Physics_
    simulationSpace=BaseBoneSpace
    alphaBoolBlend {blendOption=Linear}
    baseBoneRef {boneName=root}
    componentAppliedLinearAccClamp=<3000 3000 1000>
    componentLinearAccScale=<0.1 0.1 0.2>
    componentLinearVelScale=<0.1 0.1 0.2>
    overrideWorldGravity {z=-920}
    simSpaceSettings {worldAlpha=0.4}
   }
   n45 = Blend Poses by bool(BlendPose_0=Component To Local(ComponentPose=@n43), BlendPose_1=Component To Local(ComponentPose=@n44), BlendTime_0=0.5, BlendTime_1=0.5) {blendType=Linear, transitionType=Inertialization}
   n46 = RigidBody(ComponentPose=Local To Component(LocalPose=Use cached pose 'CtrlRigCache'()), OverrideWorldGravity="0,0,-920") {
    bClampLinearTranslationLimitToRefPose=true
    bOverrideWorldGravity=true
    overridePhysicsAsset=&/M_MED_GalaxyLevel_coat_Physics_DBNO
    simulationSpace=BaseBoneSpace
    alphaBoolBlend {blendOption=Linear}
    baseBoneRef {boneName=root}
    componentAppliedLinearAccClamp=<3000 3000 1000>
    componentLinearAccScale=<0.1 0.1 0.2>
    componentLinearVelScale=<0.1 0.1 0.2>
    overrideWorldGravity {z=-920}
    simSpaceSettings {worldAlpha=0.4}
   }
   n47 = Component To Local(ComponentPose=@n46)
   n48 = RigidBody(ComponentPose=Local To Component(LocalPose=Use cached pose 'CtrlRigCache'()), OverrideWorldGravity="0,0,-920") {
    bClampLinearTranslationLimitToRefPose=true
    bOverrideWorldGravity=true
    overridePhysicsAsset=&/M_MED_GalaxyLevel_coat_Physics_FrontEnd
    simulationSpace=BaseBoneSpace
    alphaBoolBlend {blendOption=Linear}
    baseBoneRef {boneName=root}
    componentAppliedLinearAccClamp=<3000 3000 1000>
    componentLinearAccScale=<0.1 0.1 0.2>
    componentLinearVelScale=<0.1 0.1 0.2>
    overrideWorldGravity {z=-920}
    simSpaceSettings {worldAlpha=0.4}
   }
   n49 = RigidBody(ComponentPose=Local To Component(LocalPose=Use cached pose 'CtrlRigCache'()), OverrideWorldGravity="0,0,-920") {
    bClampLinearTranslationLimitToRefPose=true
    bOverrideWorldGravity=true
    overridePhysicsAsset=&/M_MED_GalaxyLevel_coat_Physics_SkyDiv
    simulationSpace=BaseBoneSpace
    alphaBoolBlend {blendOption=Linear}
    baseBoneRef {boneName=root}
    componentAppliedLinearAccClamp=<3000 3000 1000>
    componentLinearAccScale=<0.1 0.1 0.2>
    componentLinearVelScale=<0.1 0.1 0.2>
    overrideWorldGravity {z=-920}
    simSpaceSettings {worldAlpha=0.4}
   }
   n50 = Blend Poses by bool(
    BlendPose_0=Component To Local(ComponentPose=@n49)
    BlendPose_1=Blend Poses by bool(
     BlendPose_0=Component To Local(ComponentPose=@n48)
     BlendPose_1=Blend Poses by bool(BlendPose_0=@n47, BlendPose_1=@n45, BlendTime_0=1, BlendTime_1=1) {blendType=Linear, transitionType=Inertialization}
     BlendTime_0=1
     BlendTime_1=1
    ) {blendType=Linear, transitionType=Inertialization}
    BlendTime_0=1
    BlendTime_1=1
   ) {blendType=Linear, transitionType=Inertialization}
   n51 = RigidBody(ComponentPose=Local To Component(LocalPose=Use cached pose 'CtrlRigCache'()), OverrideWorldGravity="0,0,0") {
    bClampLinearTranslationLimitToRefPose=true
    bOverrideWorldGravity=true
    overridePhysicsAsset=&/M_MED_GalaxyLevel_coat_sliding_Physics
    simulationSpace=BaseBoneSpace
    alphaBoolBlend {blendOption=Linear}
    baseBoneRef {boneName=pelvis}
    componentAppliedLinearAccClamp=<1000 1000 1000>
    componentLinearAccScale=<0.1 0.1 0.2>
    componentLinearVelScale=<0.1 0.1 0.2>
    simSpaceSettings {maxAngularVelocity=2, maxLinearAcceleration=700, maxLinearVelocity=500, velocityScaleZ=0.1, worldAlpha=0.3}
   }
   n52 = Component To Local(ComponentPose=@n51)
   n53 = Get RbanSwitch() -> (RbanSwitch:bool)
   n54 = RigidBody(
    ComponentPose=Local To Component(LocalPose=Blend Poses by bool(BlendPose_0=@n52, BlendPose_1=@n50, BlendTime_0=1, BlendTime_1=2, bActiveValue=@n53) {blendType=Linear, transitionType=Inertialization})
    OverrideWorldGravity="0,0,0"
   ) {
    bClampLinearTranslationLimitToRefPose=true
    bOverrideWorldGravity=true
    overridePhysicsAsset=&/M_MED_GalaxyLevel_necklace_Physics
    simulationSpace=BaseBoneSpace
    alphaBoolBlend {blendOption=Linear}
    baseBoneRef {boneName=root}
    componentAppliedLinearAccClamp=<3000 3000 3000>
    componentLinearAccScale=<0.1 0.1 0.3>
    componentLinearVelScale=<0.1 0.1 0.3>
    simSpaceSettings {maxAngularAcceleration=10, maxAngularVelocity=10, velocityScaleZ=0.5, worldAlpha=0.7}
   }
   n55 = Control Rig!(Source=Component To Local(ComponentPose=@n54), Alpha) -> (Pose) {
    alphaBoolBlend {blendOption=Linear}
    controlRigAssetReference {blueprintRigClass=&/M_MED_GalaxyLevel_necklacepost_CtrlRig.M_MED_GalaxyLevel_necklacepost_CtrlRig_C}
   }
   n56 = "Constraint - dyn_button_01_l (PPPPP)"!(
    ComponentPose=Local To Component(LocalPose=@n55)
    Alpha=1
    ConstraintWeights_0=0.1
    ConstraintWeights_1:float=0.05
    ConstraintWeights_2:float=0.1
    ConstraintWeights_3:float=0.62
    ConstraintWeights_4:float=0.13
   ) -> (Pose) {
    alphaBoolBlend {blendOption=Linear}
    boneToModify {boneName=dyn_button_01_l}
    constraintSetup [
     {targetBone {boneName=spine_01}}
     {targetBone {boneName=spine_02}}
     {targetBone {boneName=pelvis}}
     {targetBone {boneName=dyn_coat_01_L_01}}
     {targetBone {boneName=dyn_coat_02_L_01}}
    ]
    constraintWeights [1,1,1,1,1]
   }
   n57 = "Constraint - dyn_button_02_l (PPP)"!(ComponentPose=@n56, Alpha=1, ConstraintWeights_0=0.3, ConstraintWeights_1:float=0.55, ConstraintWeights_2:float=0.15) -> (Pose) {
    alphaBoolBlend {blendOption=Linear}
    boneToModify {boneName=dyn_button_02_l}
    constraintSetup [{targetBone {boneName=dyn_coat_01_L_01}}, {targetBone {boneName=dyn_coat_02_L_01}}, {targetBone {boneName=dyn_coat_01_L_02}}]
    constraintWeights [1,1,1]
   }
   n58 = "Constraint - dyn_button_01_r (PPPPP)"!(
    ComponentPose=@n57
    Alpha=1
    ConstraintWeights_0=0.1
    ConstraintWeights_1:float=0.05
    ConstraintWeights_2:float=0.1
    ConstraintWeights_3:float=0.6
    ConstraintWeights_4:float=0.15
   ) -> (Pose) {
    alphaBoolBlend {blendOption=Linear}
    boneToModify {boneName=dyn_button_01_r}
    constraintSetup [
     {targetBone {boneName=spine_01}}
     {targetBone {boneName=spine_02}}
     {targetBone {boneName=pelvis}}
     {targetBone {boneName=dyn_coat_01_R_01}}
     {targetBone {boneName=dyn_coat_02_R_01}}
    ]
    constraintWeights [1,1,1,1,1]
   }
   n59 = "Constraint - dyn_button_02_r (PPPP)"!(
    ComponentPose=@n58
    Alpha=1
    ConstraintWeights_0=0.01
    ConstraintWeights_1:float=0.31
    ConstraintWeights_2:float=0.1
    ConstraintWeights_3:float=0.58
   ) -> (Pose) {
    alphaBoolBlend {blendOption=Linear}
    boneToModify {boneName=dyn_button_02_r}
    constraintSetup [
     {targetBone {boneName=pelvis}}
     {targetBone {boneName=dyn_coat_01_R_01}}
     {targetBone {boneName=dyn_coat_01_R_02}}
     {targetBone {boneName=dyn_coat_02_R_01}}
    ]
    constraintWeights [1,1,1,1]
   }
   n60 = "Constraint - dyn_button_01_l (RR)"!(ComponentPose=@n59, Alpha=1, ConstraintWeights_0=1, ConstraintWeights_1:float=0) -> (Pose) {
    alphaBoolBlend {blendOption=Linear}
    boneToModify {boneName=dyn_button_01_l}
    constraintSetup [{transformType=Rotation, targetBone {boneName=dyn_coat_01_L_01}}, {transformType=Rotation, targetBone {boneName=dyn_coat_02_L_01}}]
    constraintWeights [1,1]
   }
   n61 = "Constraint - dyn_button_02_l (RR)"!(ComponentPose=@n60, Alpha=0, ConstraintWeights_0=0, ConstraintWeights_1:float=1) -> (Pose) {
    alphaBoolBlend {blendOption=Linear}
    boneToModify {boneName=dyn_button_02_l}
    constraintSetup [{transformType=Rotation, targetBone {boneName=dyn_coat_01_L_01}}, {transformType=Rotation, targetBone {boneName=dyn_coat_02_L_01}}]
    constraintWeights [1,1]
   }
   n62 = "Transform (Modify) Bone - Bone: dyn_button_02_r"(ComponentPose=@n61, Alpha=1, Translation="-0.5,0,0", Rotation="0.002,-4.745,-0.017", Scale="1,1,1") {
    rotationMode=BMM_Additive
    rotationSpace=BCS_BoneSpace
    scaleMode=BMM_Additive
    translationMode=BMM_Additive
    translationSpace=BCS_BoneSpace
    alphaBoolBlend {blendOption=Linear}
    boneToModify {boneName=dyn_button_02_r}
    rotation=<rot 0.002 -4.745 -0.017>
    translation {x=-0.5}
   }
   Output Pose(Result=Blend Poses by bool(
     BlendPose_0=Component To Local(ComponentPose=@n62)
     BlendPose_1=Blend Poses by bool(BlendPose_0=@n24, BlendPose_1=@n42, BlendTime_0=0.5, BlendTime_1=0.5) {blendType=Linear, transitionType=Inertialization}
     bActiveValue=Get bRigidBodyEnabled()
    ) {blendType=Linear}) {layerGroup=None}
   n63 = Control Rig(Source=Copy Pose From Mesh() {bUseAttachedParent=true}) {
    alphaBoolBlend {blendOption=Linear}
    controlRigAssetReference {blueprintRigClass=&/M_MED_DeformRig_GalaxyLevel_ControlRig.M_MED_DeformRig_GalaxyLevel_ControlRig_C}
   }
   n64 = Control Rig!(
    Source=@n63
    Alpha
    DinoHead:bool=false
    IsSlidingCtrl:bool=false
    IsCrouchingCtrl:bool=false
    IsGlidingCtrl:bool=Get bIsSkyDiving()
    IsScaleCtrl:bool=false
   ) -> (Pose) {
    alphaBoolBlend {blendOption=Linear}
    controlRigAssetReference {blueprintRigClass=&M_MED_GalaxyLevel_CtrlRig_C}
   }
   Save cached pose 'CtrlRigCache'(Pose=@n64) {cacheName=CtrlRigCache}
  ]}
 EventGraph {nodes [
   Delay(Duration=0.1)
   > ResetDynamics(InTeleportType=TeleportPhysics)
   > ResetDynamics(InTeleportType=ResetPhysics)
   Event BlueprintUpdateAnimation()
   > n1 = Sequence!(execute) -> (then_0, then_1, then_2:Exec, then_3:Exec, then_4:Exec, then_5:Exec, then_6:Exec, then_7:Exec)
   > n2 = Gravity Override State Function DEPRECATED(CharacterPartAnimInstance=Self-Reference(), GameTimeInSeconds=GetGameTimeInSeconds(), GravityOverrideParams=Get Gravity Override Params())
   > Set Gravity Override(Gravity Override=@n2)
   n3 = Cast To PlayerPawn_Athena(execute=@n1.then_1, Object=TryGetPawnOwner())
   > n4 = Set Head_Mesh(Head_Mesh:Skeletal Mesh Component Object Reference=GetSkeletalMeshForPartType(self=@n3, PartType=Face)) -> (Output_Get:Skeletal Mesh Component Object Reference)
   n5 = GetDisplayName(Object=@n4)
   n6 = Append(A=GetObjectName(Object=@n4), B=" ")
   > Set DinoHead?(DinoHead?:bool="Equal, Case Insensitive (String)"(A=Split(SourceString=@n5, InStr=@n6, SearchDir=FromStart).RightS, B=M_MED_GalaxyLevel_FaceAcc)) -> (Output_Get:bool)
   n7 = Get bIsSliding()
   Set NormalRbanAlpha(execute=@n1.then_2, NormalRbanAlpha:double=SelectFloat(A=0, B=1, bPickA=@n7)) -> (Output_Get:double)
   > Set RbanSwitch(RbanSwitch:bool=@n7) -> (Output_Get:bool)
   > Set RbanOn?(RbanOn?:bool=Get bRigidBodyEnabled()) -> (Output_Get:bool)
   n8 = Get bIsSliding()
   Set CrouchAlpha(execute=@n1.then_3, CrouchAlpha:bool=AND Boolean(A=Get bIsCrouching(), B="Not Equal (Boolean)"(A=@n8, B=true) -> (ReturnValue=~))) -> (Output_Get:bool)
   > n9 = Branch(Condition=Get bIsSliding())
   Do Once(execute=@n9.else, Reset=@n9.then, Start Closed=true)
   > Delay(Duration=0.2)
   > ResetDynamics(InTeleportType=TeleportPhysics)
   > ResetDynamics(InTeleportType=ResetPhysics)
   n10 = Get Skel_Mesh()
   Set thigh_rot_otr_l(
    execute=@n1.then_4
    thigh_rot_otr_l:double=GetAlphaFromReferenceTransformDelta(SkeletalMeshComponent=@n10, BoneName=thigh_l, BaseName=pelvis, DeltaComponentType=EulerY, InputRange=(X=0,Y=30), TwistAxis="1,0,0")
   ) -> (Output_Get:double)
   > Set thigh_rot_otr_r(thigh_rot_otr_r:double=GetAlphaFromReferenceTransformDelta(SkeletalMeshComponent=@n10, BoneName=thigh_r, BaseName=pelvis, DeltaComponentType=EulerY, InputRange=(X=0,Y=40), TwistAxis="1,0,0")) -> (Output_Get:double)
   n11 = Get Skel_Mesh()
   > Set thigh_rot_inr_l(thigh_rot_inr_l:double=GetAlphaFromReferenceTransformDelta(SkeletalMeshComponent=@n11, BoneName=thigh_l, BaseName=pelvis, DeltaComponentType=EulerY, InputRange=(X=10,Y=-6), TwistAxis="1,0,0")) -> (Output_Get:double)
   > Set thigh_rot_inr_r(thigh_rot_inr_r:double=GetAlphaFromReferenceTransformDelta(SkeletalMeshComponent=@n11, BoneName=thigh_r, BaseName=pelvis, DeltaComponentType=EulerY, InputRange=(X=0,Y=-10), TwistAxis="1,0,0")) -> (Output_Get:double)
   n12 = Get Skel_Mesh()
   > Set thigh_rot_bck_r(thigh_rot_bck_r:double=float + float(B=GetAlphaFromReferenceTransformDelta(SkeletalMeshComponent=@n12, BoneName=thigh_r, BaseName=pelvis, InputRange=(X=0,Y=10), TwistAxis="1,0,0").RawOutput)) -> (Output_Get:double)
   n13 = Branch(execute=@n1.then_5, Condition=Get bIsInFrontEnd())
   > Set BwdMovement(BwdMovement:double=0) -> (Output_Get:double)
   > Set RightMovement(RightMovement:float=0) -> (Output_Get:float)
   > Set LeftMovement(LeftMovement:float=0) -> (Output_Get:float)
   n14 = PlayerMovement(execute=@n13.else, AnimInstance=Self-Reference(), XDotProductCoefficient=1, YDotProductCoefficient=1, ZDotProductCoefficient=1)
   > n15 = Branch(Condition=Get bIsCrouching())
   n16 = MapRangeClamped(Value=@n14.YMovement, InRangeA=0, InRangeB=-1)
   n17 = float * float(A="Absolute (Float)"(A=@n14), B=0.5)
   > Set BwdMovement(BwdMovement=MapRangeClamped(
     Value=float + float(A=@n17, B=@n16) {is=title K2Node_CommutativeAssociativeBinaryOperator}
     InRangeA=0
     InRangeB=1.1
     OutRangeA=0.8
     OutRangeB=0
    ))
   > Set FwdMovement(FwdMovement:double=MapRangeClamped(Value=@n14.YMovement, InRangeA=1, InRangeB=0)) -> (Output_Get:double)
   n18 = float + float(A=float * float(A="Absolute (Float)"(A=@n14), B=0.5), B=MapRangeClamped(Value=@n14.YMovement, InRangeA=0, InRangeB=-1))
   Set BwdMovement(execute=@n15.else, BwdMovement=MapRangeClamped(Value=@n18, InRangeA=0, InRangeB=1.1, OutRangeA=0.8, OutRangeB=0))
   > Set FwdMovement(FwdMovement=MapRangeClamped(Value=@n18, InRangeA=0, InRangeB=1.1))
   n19 = MapRangeClamped(Value=@n14, InRangeA=-0.2, InRangeB=1)
   n20 = float * float(A="Absolute (Float)"(A=@n14.YMovement), B=0.2)
   > Set RightMovement(RightMovement=MapRangeClamped(Value=float + float(A=@n20, B=@n19), InRangeA=0, InRangeB=1.2, OutRangeB=0.7))
   n21 = MapRangeClamped(Value=@n14, InRangeA=0.2, InRangeB=-1, OutRangeB=0.3)
   n22 = float * float(A="Absolute (Float)"(A=@n14.YMovement), B=0.2)
   > Set LeftMovement(LeftMovement=MapRangeClamped(Value=float + float(A=@n22, B=@n21), InRangeA=0, InRangeB=1.2, OutRangeB=0.7))
   n23 = Get bIsSkyDiving()
   Branch(execute=@n1.then_6, Condition=AND Boolean(A=Get bIsFalling(), B="Not Equal (Boolean)"(A=@n23, B=true) -> (ReturnValue=~)))
   > Set BwdMovement(BwdMovement=float * float(A=Get BwdMovement(), B=0.7) {is=title K2Node_PromotableOperator})
   n24 = Branch(execute=@n1.then_7, Condition=Get bIsSkyDiving())
   > n25 = Branch(Condition="float > float"(A=Get PawnVelocityZ(), B=50) {is=title K2Node_CallFunction})
   > Set BwdMovement(BwdMovement=float * float(A=Get BwdMovement(), B=0) {is=title K2Node_CommutativeAssociativeBinaryOperator})
   Set BwdMovement(execute=[@n25.else, @n24.else], BwdMovement=float * float(A=Get BwdMovement(), B=1))
   Event BlueprintInitializeAnimation()
   > Set Skel_Mesh(Skel_Mesh=GetOwnerSkeletalMeshComponent())
   Get Distance Between Two Sockets!(
    Component:Skeletal Mesh Component Object Reference=@n12
    SocketOrBoneNameA:Name=dyn_coat_03_R_03
    SocketSpaceA:ERelativeTransformSpace Enum=RTS_ParentBoneSpace
    SocketOrBoneNameB:Name=dyn_coat_03_R_04
    SocketSpaceB:ERelativeTransformSpace Enum=RTS_ParentBoneSpace
    bRemapRange:bool=false
    InRangeMin:float=0
    InRangeMax:float=50
    OutRangeMin:float=0
    OutRangeMax:float=1
   ) -> (ReturnValue:float) {is=title K2Node_CallFunction}
   GetAlphaFromReferenceTransformDelta(
    SkeletalMeshComponent=@n12
    BoneName=dyn_coat_03_R_01
    BaseName=dyn_coat_03_R_mstr
    DeltaComponentType=EulerY
    InputRange=(X=8,Y=14)
    TwistAxis="1,0,0"
   )
   n26 = GetDeltaTransformFromRefPose(self=GetOwnerSkeletalMeshComponent(), BoneName=dyn_coat_03_R_01, BaseName=pelvis)
   MapRangeClamped(Value=BreakRotator(InRot=BreakTransform(InTransform=@n26).Rotation).Pitch, InRangeA=0, InRangeB=30)
   GetAlphaFromReferenceTransformDelta(SkeletalMeshComponent=@n12, BoneName=thigh_l, BaseName=pelvis, DeltaComponentType=EulerY, InputRange=(X=10,Y=-6), TwistAxis="1,0,0")
   OR Boolean(A=@n7, B=Get bIsCrouching())
   Event OnCharacterCustomizationLoaded()
   Set DinoHead?(DinoHead?=CharacterItemDefinitionHasAnyMetaTag(Tags=Make Array("[0]"=(TagName="Cosmetics.MetaTags.Style.OptionA"))))
  ]}
]
variables [
 BwdMovement {type=number}
 CoatAlpha01 {type=number}
 CrouchAlpha {type=boolean}
 DinoHead? {type=boolean}
 FwdMovement {type=number}
 Gravity Override {type=Vector}
 Gravity Override Params {type=GravityOverrideParamsStruct, default {skydive_Surfing {gravityOverride=<0 -300 -980>}}}
 Head_Mesh {type=SkeletalMeshComponent}
 Lapel L alpha {type=number}
 LeftMovement {type=number}
 NormalRbanAlpha {type=number}
 RbanOn? {type=boolean}
 RbanSwitch {type=boolean}
 RightMovement {type=number}
 Skel_Mesh {type=SkeletalMeshComponent}
 thigh_rot_bck_r {type=number}
 thigh_rot_inr_l {type=number}
 thigh_rot_inr_r {type=number}
 thigh_rot_otr_l {type=number}
 thigh_rot_otr_r {type=number}
]
